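# Jam session song-info HUD panel showing audio state, the local player's loop, and privacy or jammer-limit messages.
# Session-member gameplay tags on the PlayerState start or tear down session discovery through the notifier component.
# Loop, privacy and persistent-party events drive refreshes, and part-switcher events hide it in condensed view.
# Rebuild updates only the personal display; loop-list helpers exist, but the rebuild never calls them.
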WidgetBlueprint /FMJamSystem/Widgets/SongInfo/WBP_JamSongInfo
graphs [
 Add Loop Display {nodes [
   n1 = Add Loop Display!() -> (
    then
    Song:Catalog Data Object Reference
    LoopType:EFMJamLoopType Enum=Lead
    PlayerNetId:Unique Net Id Structure
    LoopInstanceId:int
    bUseAutoJammerProfilePic:bool=~
   ) {is=title K2Node_FunctionEntry}
   > n2 = Construct Object from Class(Class=&/BP_JamLoopDisplayData.BP_JamLoopDisplayData_C) -> (ReturnValue:BP Jam Loop Display Data Object Reference)
   > UpdateJamLoopDisplayListItem!(
    execute
    self
    ListItem:BP Jam Loop Display Data Object Reference=@n2
    Song:Catalog Data Object Reference=@n1
    LoopType:EFMJamLoopType Enum=@n1.LoopType
    PlayerNetId:Unique Net Id Structure=@n1.PlayerNetId
    LoopInstanceId:int=@n1.LoopInstanceId
    bUseAutoJammerProfilePic:bool=@n1.bUseAutoJammerProfilePic
   ) -> (then)
   > AddItem!(
    execute
    self:List View Object Reference=Get JamLoopDisplayList() -> (JamLoopDisplayList:List View Object Reference)
    Item:Object Reference=@n2
   ) -> (then)
  ]}
 EventGraph {nodes [
   Event OnInitialized()
   > Parent: OnInitialized() {is=title K2Node_CallParentFunction}
   > n1 = Sequence()
   n2 = Get PlayerState!(self:Controller Object Reference=GetOwningPlayer()) -> (PlayerState:Player State Object Reference)
   > n3 = WaitGameplayTagAddToActor!(
    execute:Exec
    TargetActor:Actor Object Reference=@n2
    Tag:Gameplay Tag Structure=(TagName="FMJam.Session.Member.Client")
    OnlyTriggerOnce:bool=false
   ) -> (then:Exec, AsyncTaskProxy:Ability Async Wait Gameplay Tag Added Object Reference, Added:Exec) {is=title K2Node_AsyncAction}
   > Set ParticipantAddedEventHandle(ParticipantAddedEventHandle:Ability Async Wait Gameplay Tag Added Object Reference=@n3) -> (Output_Get:Ability Async Wait Gameplay Tag Added Object Reference)
   OnParticipantTagAdded(execute=@n3.Added)
   n4 = WaitGameplayTagRemoveFromActor!(
    execute:Exec=@n1.then_1
    TargetActor:Actor Object Reference=@n2
    Tag:Gameplay Tag Structure=(TagName="FMJam.Session.Member.Client")
    OnlyTriggerOnce:bool=false
   ) -> (then:Exec, AsyncTaskProxy:Ability Async Wait Gameplay Tag Removed Object Reference, Removed:Exec) {is=title K2Node_AsyncAction}
   > Set ParticipantRemovedEventHandle(ParticipantRemovedEventHandle:Ability Async Wait Gameplay Tag Removed Object Reference=@n4) -> (Output_Get:Ability Async Wait Gameplay Tag Removed Object Reference)
   OnParticipantTagRemoved(execute=@n4.Removed)
   Event Destruct()
   > OnParticipantTagRemoved()
   > n5 = Sequence!(execute) -> (then_0, then_1, then_2:Exec)
   n6 = Get ParticipantAddedEventHandle() -> (ParticipantAddedEventHandle:Ability Async Wait Gameplay Tag Added Object Reference)
   > Is Valid(InputObject=@n6)
   > Cancel(self:Cancellable Async Action Object Reference=@n6)
   > Set ParticipantAddedEventHandle()
   n7 = Get ParticipantRemovedEventHandle() -> (ParticipantRemovedEventHandle:Ability Async Wait Gameplay Tag Removed Object Reference)
   Is Valid(exec=@n5.then_1, InputObject=@n7)
   > Cancel(self=@n7)
   > Set ParticipantRemovedEventHandle()
   n8 = Get JamSession() -> (JamSession:Jam Session Interface Interface)
   Branch(execute=@n5.then_2, Condition=IsValidInterface!(Interface:"Interface Interface (by ref)"=@n8) -> (ReturnValue:bool))
   > Unbind Event from On Persistent Party Changed(
    self:Jam Session Component Player Manager Object Reference=GetJamPlayerManagerComponent!(self:Jam Session Interface Interface=@n8) -> (ReturnValue:Jam Session Component Player Manager Object Reference)
    Delegate=Create Event()
   ) {is=title K2Node_RemoveDelegate}
  ]}
 GetDisplayListItemForLoopInstanceId {nodes [
   n1 = GetDisplayListItemForLoopInstanceId!() -> (then, LoopInstanceId:int) {is=title K2Node_FunctionEntry}
   > n2 = For Each Loop(Array:Array of Object References=GetListItems!(self:List View Object Reference=Get JamLoopDisplayList()) -> (ReturnValue:Array of Object References)) -> (Array Element:Object Reference)
   Return Node!(execute=@n2.Completed, DisplayData:BP Jam Loop Display Data Object Reference)
   > n3 = Cast To BP_JamLoopDisplayData!(execute, Object=@n2) -> (then, CastFailed, AsBP Jam Loop Display Data:BP Jam Loop Display Data Object Reference) {is=title K2Node_DynamicCast}
   n4 = Get DisplayData!(self:BP Jam Loop Display Data Object Reference=@n3) -> (DisplayData:S Jam Loop Display Data Structure)
   > Branch(Condition="Equal (Integer)"(
     A=Break S Jam Loop Display Data!(S_JamLoopDisplayData:"S Jam Loop Display Data Structure (by ref)"=@n4) -> (
      Title_3_#1:Text
      Artist_21_#2:Text
      LoopType_6_#3:EFMJamLoopType Enum
      PlayerNetId_26_#4:Unique Net Id Structure
      LoopInstanceId_17_#5:int
      bUseAutoJammerProfilePic_23_#6:bool=~
     ) {is=title K2Node_BreakStruct}.LoopInstanceId_17_#5
     B=@n1
    ))
   > Return Node(DisplayData=@n3)
  ]}
 IsPlayerAllowedToJoinClosedJam {nodes [
   n1 = IsPlayerAllowedToJoinClosedJam!() -> (then, FortPawn:Fort Pawn Object Reference) {is=title K2Node_FunctionEntry}
   > n2 = Sequence!(execute) -> (then_0, then_1)
   n3 = GetMusicSlotsInUse!(self:Jam Session Component Music Manager Object Reference=GetMusicManager!(self:Jam Session Interface Interface=Get JamSession()) -> (ReturnValue:Jam Session Component Music Manager Object Reference)) -> (ReturnValue:Array of Jam Music Slot Object References)
   > n4 = Branch(Condition=Is Not Empty!(TargetArray:Array of Jam Music Slot Object References=@n3) -> (ReturnValue:bool))
   > n5 = For Each Loop with Break!(Exec:Exec, Array:Array of Jam Music Slot Object References=@n3, Break:Exec=@n11.then) -> (LoopBody:Exec, Array Element:Jam Music Slot Object Reference, Array Index:int, Completed:Exec)
   > n6 = Cast To FortPlayerPawnAthena(Object=@n1)
   n7 = Break Jam Play Params!(JamPlayParams:"Jam Play Params Structure (by ref)"=GetPlayParams!(self:Jam Music Slot Object Reference=@n5) -> (ReturnValue:"Jam Play Params Structure (by ref)")) -> (
    SongShortName:Name
    LoopType:EFMJamLoopType Enum
    PlayerNetId:Unique Net Id Structure
    LoopInstanceId:int
    bIsAutoJammer:bool=~
    PlayingActorOverride:Actor Weak object
    ServerQuantizationTimeSeconds:float=~
    SlotStartMethod:EJamSlotStartMethod Enum
   ) {is=title K2Node_BreakStruct}
   > n8 = Cast To FortPlayerPawnAthena(Object=GetPawnFromActor!(
     Actor:Actor Object Reference=GetPlayerStateFromUniqueNetId!(UniqueId:"Unique Net Id Structure (by ref)"=@n7.PlayerNetId) -> (ReturnValue:Player State Object Reference)
     PawnType:Pawn Class Reference
    ) -> (ReturnValue:Pawn Object Reference))
   n9 = OnSameSquad!(PlayerPawnA:Fort Player Pawn Athena Object Reference=@n8, PlayerPawnB:Fort Player Pawn Athena Object Reference=@n6) -> (ReturnValue:bool)
   n10 = OnSameTeam!(ActorA:Actor Object Reference=@n8, ActorB:Actor Object Reference=@n6) -> (ReturnValue:bool)
   > Branch(Condition=Select(
     Option 0:bool=IsPlayerInLocalParty!(OtherPlayerPawn:Fort Player Pawn Athena Object Reference=@n8) -> (ReturnValue:bool)
     Option 1:bool=OR Boolean(A=@n10, B=@n9)
     Index=InEditorOrPIE!() -> (ReturnValue:bool)
    ) -> (ReturnValue:bool))
   > n11 = Set CanJoin(CanJoin:bool=true) -> (Output_Get:bool)
   Set CanJoin(execute=@n4.else, CanJoin=false)
   Return Node!(execute=@n2.then_1, AllowedToJoin:bool=Get CanJoin() -> (CanJoin:bool))
  ]}
 PlayspaceEventGraph {nodes [
   OnParticipantTagAdded() {is=title K2Node_CustomEvent}
   > RebuildLoopDisplayList()
   > n1 = Set LastCheckedNotifierComp(LastCheckedNotifierComp:Actor Component Jam Session Notifier Object Reference=GetComponentByClass!(
     self:Actor Object Reference=GetOwningPlayer()
     ComponentClass:Actor Component Class Reference=/Script/FMJamSessionRuntime.ActorComponent_JamSessionNotifier
    ) -> (ReturnValue:Actor Component Jam Session Notifier Object Reference)) -> (Output_Get:Actor Component Jam Session Notifier Object Reference)
   > Is Valid(InputObject=@n1)
   > n2 = Get Current JamSession!(
    execute
    self:Jam Session Observer Interface Interface=Get LastCheckedNotifierComp() -> (LastCheckedNotifierComp:Actor Component Jam Session Notifier Object Reference)
   ) -> (True:Exec, False:Exec, ReturnValue:Jam Session Interface Interface) {is=title K2Node_CallFunction}
   n3 = Found Jam Session!() -> (OutputDelegate, then, EventType:EJamSessionNotificationType Enum=EnteringSession, JamSession:Jam Session Interface Interface) {is=title K2Node_CustomEvent}
   Bind Event to On Jam Session Changed BP(execute=[@n2.False, @n4.then], self:Actor Component Jam Session Notifier Object Reference=Get LastCheckedNotifierComp(), Delegate=@n3) {is=title K2Node_AddDelegate}
   n4 = Found Jam Session!(execute=@n2.True, self, EventType:EJamSessionNotificationType Enum=EnteringSession, JamSession:Jam Session Interface Interface=@n2) -> (then) {is=title K2Node_CallFunction}
   OnParticipantTagRemoved() {is=title K2Node_CustomEvent}
   > SetVisibility(InVisibility=Collapsed)
   > n5 = Sequence!(execute) -> (then_0, then_1, then_2:Exec, then_3:Exec)
   n6 = Get Privacy Changed Event Handle() -> (Privacy Changed Event Handle:Async Action Start Listening To Stateful Event Object Reference)
   > Is Valid(InputObject=@n6)
   > StopListeningToStatefulEvent(self:Async Action Start Listening To Stateful Event Object Reference=@n6)
   > Set Privacy Changed Event Handle(Privacy Changed Event Handle:Async Action Start Listening To Stateful Event Object Reference) -> (Output_Get:Async Action Start Listening To Stateful Event Object Reference)
   Sequence!(execute=@n5.then_1) -> (then_0, then_1)
   n7 = Get JamLoopChangedEventHandle() -> (JamLoopChangedEventHandle:Async Action Start Listening To Event Object Reference)
   > Is Valid(InputObject=@n7)
   > StopListeningToEvent(self:Async Action Start Listening To Event Object Reference=@n7)
   > Set JamLoopChangedEventHandle(JamLoopChangedEventHandle:Async Action Start Listening To Event Object Reference) -> (Output_Get:Async Action Start Listening To Event Object Reference)
   Set GameplayEventRouter(execute=@n5.then_3, GameplayEventRouter:Gameplay Event Router Component Object Reference) -> (Output_Get:Gameplay Event Router Component Object Reference)
   > Set JamPlayspace(JamPlayspace:Jam Playspace Object Reference) -> (Output_Get:Jam Playspace Object Reference)
   > Set JamSession(JamSession:Jam Session Interface Interface) -> (Output_Get:Jam Session Interface Interface)
   n8 = Get PartSwitcherEventHandle() -> (PartSwitcherEventHandle:Async Action Start Listening To Event Object Reference)
   Is Valid(exec=@n5.then_2, InputObject=@n8)
   > StopListeningToEvent(self=@n8)
   > Set PartSwitcherEventHandle(PartSwitcherEventHandle:Async Action Start Listening To Event Object Reference) -> (Output_Get:Async Action Start Listening To Event Object Reference)
   OnPrivacyChanged() {is=title K2Node_CustomEvent}
   > UpdateJamStatusDisplay()
   OnJamLoopsChanged() {is=title K2Node_CustomEvent}
   > UpdateJamStatusDisplay()
   > RebuildLoopDisplayList()
   Bind Event to On Persistent Party Changed(
    execute=@n22.then_4
    self:Jam Session Component Player Manager Object Reference=GetJamPlayerManagerComponent(self=Get JamSession())
    Delegate=Create Event()
   ) {is=title K2Node_AddDelegate}
   OnPrivacyChanged(execute=[@n12.OnEventReceived, @n12.OnSavedEventStateExists]) {is=title K2Node_CallFunction}
   RebuildLoopDisplayList(execute=@n21.then_0)
   FortShippingLog!(execute=@n25.else, WorldContextObject:Object Reference=Self-Reference(), InString:String=Exited Jam Session?, bLogAsWarning:bool=true) -> (then)
   n9 = CreateJamGlobalStateVM!(execute=@n22.then_0, JamSession:"Jam Session Interface Interface (by ref)"=Get JamSession()) -> (then, ReturnValue:Jam Global State VM Object Reference)
   > n10 = Set JamGlobalStateVM(JamGlobalStateVM:Jam Global State VM Object Reference=@n9) -> (Output_Get:Jam Global State VM Object Reference)
   > SetJamGlobalStateVM!(
    execute
    self:WBP Jam Audio State Display Object Reference=Get WBP_JamAudioStateDisplay() -> (WBP_JamAudioStateDisplay:WBP Jam Audio State Display Object Reference)
    Viewmodel:Jam Global State VM Object Reference=@n10
   ) -> (then)
   n11 = StartListeningToEvent!(
    execute:Exec=@n15.then
    Target:Gameplay Event Router Component Object Reference=Get GameplayEventRouter() -> (GameplayEventRouter:Gameplay Event Router Component Object Reference)
    EventType:Script Struct Object Reference=/FMJamSystem/Widgets/PartSwitcher/JamEvent_PartSwitcher_Displayed
    Context:Object Reference
    EventBubblingRule:EEventBubblingRule Enum
   ) -> (
    then:Exec
    AsyncTaskProxy:Async Action Start Listening To Event Object Reference
    OnEventReceived:Exec
    Payload:Jam Event Part Switcher Displayed Structure
   )
   SetVisibility(
    execute=@n11.OnEventReceived
    InVisibility=Select(
     Option 0:ESlateVisibility Enum=SelfHitTestInvisible
     Option 1:ESlateVisibility Enum=Collapsed
     Index=Break Jam Event Part Switcher Displayed!(JamEvent_PartSwitcher_Displayed:"Jam Event Part Switcher Displayed Structure (by ref)"=@n11.Payload) -> (isDisplayed_1_#7:bool=~) {is=title K2Node_BreakStruct}
    ) -> (ReturnValue:ESlateVisibility Enum)
   )
   FortShippingLog(execute=@n23.else, WorldContextObject=Self-Reference(), InString=Found invalid JamSession?, bLogAsWarning=true)
   OnJamLoopsChanged(execute=@n13.OnEventReceived) {is=title K2Node_CallFunction}
   Reroute Node(InputPin:Wildcard) -> (OutputPin:Wildcard)
   Set PartSwitcherEventHandle(execute=@n11.then, PartSwitcherEventHandle=@n11)
   n12 = StartListeningToStatefulEvent!(
    execute:Exec=@n22.then_1
    Target:Gameplay Event Router Component Object Reference=GetGameplayEventRouterComponent!(self:Jam Session Interface Interface=Get JamSession()) -> (ReturnValue:Gameplay Event Router Component Object Reference)
    EventType:Script Struct Object Reference=/Script/FMJamSessionRuntime.JamEvent_PrivacyChanged
    Context:Object Reference
    EventBubblingRule:EEventBubblingRule Enum
   ) -> (
    then:Exec
    AsyncTaskProxy:Async Action Start Listening To Stateful Event Object Reference
    OnEventReceived:Exec
    OnSavedEventStateExists:Exec
    OnEventStateCleared:Exec
    Payload:Jam Event Privacy Changed Structure
   ) {is=title K2Node_StartListeningToStatefulEventAsyncAction}
   > Set Privacy Changed Event Handle(Privacy Changed Event Handle=@n12)
   n13 = StartListeningToEvent(execute=@n21.then_1, Target=Get GameplayEventRouter(), EventType=/Script/FMJamSessionRuntime.JamEvent_LoopsChangedThisFrame) -> (Payload:Jam Event Loops Changed This Frame Structure)
   > Set JamLoopChangedEventHandle(JamLoopChangedEventHandle=@n13)
   n14 = GetGameplayEventRouterComponent(self=Get JamSession())
   UpdateJamStatusDisplay(execute=@n18.then)
   > n15 = Branch(Condition=Get CondensedView() -> (CondensedView:bool))
   n16 = Set JamSession(execute=@n25.then, JamSession=@n3.JamSession)
   n17 = IsValidInterface(Interface=@n16)
   n18 = SetVisibility(execute=@n22.then_3, InVisibility=HitTestInvisible)
   n19 = Create Event()
   n20 = "Equal (Enum)"(A:EJamSessionNotificationType Enum=@n3.EventType, B:EJamSessionNotificationType Enum=EnteringSession)
   n21 = Sequence(execute=@n22.then_2)
   n22 = Sequence!(execute=@n24.then) -> (then_0, then_1, then_2:Exec, then_3:Exec, then_4:Exec)
   n23 = Branch(execute=@n16.then, Condition=@n17)
   > Unbind Event from On Jam Session Changed BP(self:Actor Component Jam Session Notifier Object Reference=Get LastCheckedNotifierComp(), Delegate=@n19) {is=title K2Node_RemoveDelegate}
   > Is Valid(InputObject=@n14)
   > n24 = Set GameplayEventRouter(GameplayEventRouter=@n14)
   n25 = Branch(execute=@n3.then, Condition=@n20)
  ]}
 RebuildLoopDisplayList {nodes [
   RebuildLoopDisplayList() {is=title K2Node_FunctionEntry}
   > Branch(Condition=IsValidInterface(Interface=Get JamSession()))
   > n1 = Cast To FortPlayerController(Object=GetOwningPlayer())
   n2 = Get JamSession()
   > n3 = Set LocalPlayerMusicSlot(LocalPlayerMusicSlot:Jam Music Slot Object Reference=GetMusicSlotForPlayerId!(
     self:Jam Session Component Music Manager Object Reference=GetMusicManager(self=@n2)
     PlayerNetId:"Unique Net Id Structure (by ref)"=GetGameAccountId!(self:Fort Player Controller Object Reference=@n1) -> (ReturnValue:Unique Net Id Structure)
    ) -> (ReturnValue:Jam Music Slot Object Reference)) -> (Output_Get:Jam Music Slot Object Reference)
   > n4 = Is Valid(InputObject=@n3)
   ShowOrUpdatePersonalSongInfo!(execute=@n4.Is Not Valid, self, Song:Catalog Data Object Reference, InputPin:EFMJamLoopType Enum=Lead) -> (then)
   ShowOrUpdatePersonalSongInfo(
    execute=@n4.Is Valid
    Song=GetCurrentCatalogEntry!(self:Jam Music Slot Object Reference=@n3) -> (ReturnValue:Catalog Data Object Reference)
    InputPin=GetCurrentLoopType!(self:Jam Music Slot Object Reference=@n3) -> (ReturnValue:EFMJamLoopType Enum=Lead)
   )
  ]}
 RemoveJamLoopDisplayListItem {nodes [
   n1 = RemoveJamLoopDisplayListItem!() -> (then, LoopInstanceId:int) {is=title K2Node_FunctionEntry}
   n2 = GetDisplayListItemForLoopInstanceId!(LoopInstanceId:int=@n1) -> (DisplayData:BP Jam Loop Display Data Object Reference) {is=title K2Node_CallFunction}
   > n3 = Is Valid(InputObject=@n2)
   Return Node!(execute=@n3.Is Not Valid)
   > RemoveItem!(execute, self:List View Object Reference=Get JamLoopDisplayList(), Item:Object Reference=@n2) -> (then)
   > Return Node()
  ]}
 ShowOrUpdatePersonalSongInfo {nodes [
   n1 = ShowOrUpdatePersonalSongInfo!() -> (then, Song:Catalog Data Object Reference, InputPin:EFMJamLoopType Enum=Lead) {is=title K2Node_FunctionEntry}
   > n2 = Is Valid(InputObject=@n1)
   > n3 = Sequence!(execute) -> (then_0, then_1)
   n4 = Break Sparks Song Catalog Entry!(SparksSongCatalogEntry:"Sparks Song Catalog Entry Structure (by ref)"=Get DataEntry!(self:Catalog Data Object Reference=@n1) -> (DataEntry:Sparks Song Catalog Entry Structure)) -> (
    Title:String
    Artist:String
    AnimGenre:Gameplay Tag Structure
    StageMood:Gameplay Tag Structure
    CountInDuration:float=~
    MusicMomentOffsetSeconds:float=~
    NewUntilTime:Date Time Structure
    AgeRating:ESparksSongAgeRating Enum
    SanitizedSearchString:String
    AlbumArtThumbnailUrl:String
   ) {is=title K2Node_BreakStruct}
   > SetValues!(
    execute
    self:WBP Personal Jam Loop Display Object Reference=Get WBP_PersonalJamLoopDisplay() -> (WBP_PersonalJamLoopDisplay:WBP Personal Jam Loop Display Object Reference)
    Title:Text="To Text (String)"(InString=@n4)
    Artist:Text="To Text (String)"(InString=@n4.Artist)
    LoopType:EFMJamLoopType Enum=@n1.InputPin
   ) -> (then)
   n5 = Sequence(execute=[@n3.then_1, @n2.Is Not Valid])
   Return Node(execute=@n5.then_1)
   n6 = IsValid(Object=@n1)
   n7 = Get CondensedView()
   SetVisibility(
    self=[Get WBP_PersonalJamLoopDisplay(), Get WBP_JamAudioStateDisplay()]
    InVisibility=Select(Option 0=Collapsed, Option 1=SelfHitTestInvisible, Index=AND Boolean(A=@n7, B=@n6))
   )
  ]}
 UpdateJamLoopDisplayListItem {nodes [
   n1 = UpdateJamLoopDisplayListItem!() -> (
    then
    ListItem:BP Jam Loop Display Data Object Reference
    Song:Catalog Data Object Reference
    LoopType:EFMJamLoopType Enum=Lead
    PlayerNetId:Unique Net Id Structure
    LoopInstanceId:int
    bUseAutoJammerProfilePic:bool=~
   ) {is=title K2Node_FunctionEntry}
   > n2 = Sequence()
   Is Valid(exec=[@n2.then_0, @n5.then], InputObject=@n1)
   > n3 = Is Valid(InputObject=@n1.Song)
   n4 = Break Sparks Song Catalog Entry!(SparksSongCatalogEntry=Get DataEntry(self=@n1.Song)) -> (
    Title
    Artist
    AnimGenre
    StageMood
    Duration:int
    CountInDuration=~
    MusicMomentOffsetSeconds=~
    NewUntilTime
    AgeRating
    SanitizedSearchString
    AlbumArtThumbnailUrl
    LipSyncUrl:String
    bValid:bool=~
   )
   > Init!(
    execute
    self:BP Jam Loop Display Data Object Reference=@n1
    Title:Text="To Text (String)"(InString=@n4)
    Artist:Text="To Text (String)"(InString=@n4.Artist)
    LoopType:EFMJamLoopType Enum=@n1.LoopType
    PlayerNetId:Unique Net Id Structure=@n1.PlayerNetId
    LoopInstanceId:int=@n1.LoopInstanceId
    bUseAutoJammerProfilePic:bool=@n1.bUseAutoJammerProfilePic
   ) -> (then)
   Init(execute=@n3.Is Not Valid, self=@n1, LoopType=Lead, LoopInstanceId=-1, bUseAutoJammerProfilePic=false)
   Return Node(execute=@n2.then_1)
   n5 = Branch(Condition=NOT Boolean(A=Get CondensedView()))
  ]}
 UpdateJamStatusDisplay {nodes [
   UpdateJamStatusDisplay() {is=title K2Node_FunctionEntry}
   > n1 = Cast To FortPlayerController(Object=GetOwningPlayer())
   > n2 = Set LocalPlayerNetId(LocalPlayerNetId:Unique Net Id Structure=GetGameAccountId(self=@n1)) -> (Output_Get:Unique Net Id Structure)
   n3 = Get JamSession()
   n4 = For Each Loop(
    Exec=[@n2.then, @n1.CastFailed]
    Array:Array of Unique Net Id Structures=GetAllJammers!(self:Jam Session Component Music Manager Object Reference=GetMusicManager(self=@n3)) -> (ReturnValue:Array of Unique Net Id Structures)
   ) -> (Array Element:Unique Net Id Structure)
   > Branch(Condition="Is Valid (UniqueIdRepl)"!(InUniqueNetIdRepl:"Unique Net Id Structure (by ref)"=@n4) -> (ReturnValue:bool))
   > Increment Int(Value=Get NumJammers_Local() -> (NumJammers_Local:int=0))
   > Branch(Condition="Equal (UniqueIdRepl)"(
     A:"Unique Net Id Structure (by ref)"=@n4
     B:"Unique Net Id Structure (by ref)"=Get LocalPlayerNetId() -> (LocalPlayerNetId:Unique Net Id Structure)
    ) -> (ReturnValue:bool=~) {is=title K2Node_PromotableOperator})
   > Set IsViewingPlayerAJammer_Local(IsViewingPlayerAJammer_Local:bool=true) -> (Output_Get:bool)
   n5 = Sequence(execute=@n4.Completed)
   n6 = Get NumJammers_Local() -> (NumJammers_Local=0)
   > Set Visibility(Visibility=Select(Option 0=Collapsed, Option 1=SelfHitTestInvisible, Index="integer > integer"(A=@n6)), self=Get WBP_JamAudioStateDisplay())
   n7 = IsJamPrivacyEnabled!(execute=@n5.then_1) -> (then, ReturnValue:bool)
   > n8 = Branch(Condition=@n7)
   n9 = GetMusicManager(self=Get JamSession())
   > n10 = Branch(Condition="Not Equal (Enum)"!(
     A:EJamPrivacySetting Enum=Get CurrentJamPrivacy!(self:Jam Session Component Music Manager Object Reference=@n9) -> (CurrentJamPrivacy:EJamPrivacySetting Enum=Public)
     B:EJamPrivacySetting Enum=PartyOnly
    ) -> (ReturnValue:bool))
   n11 = NOT Boolean(A=Get CondensedView())
   n12 = NOT Boolean(A=Get IsViewingPlayerAJammer_Local() -> (IsViewingPlayerAJammer_Local:bool))
   n13 = "integer >= integer"(A=Get NumJammers_Local() -> (NumJammers_Local=0), B=Get MaxNumJammers() -> (MaxNumJammers:int=0))
   n14 = Get StatusMessage() -> (StatusMessage:WBP Jam Now Playing Status Message Object Reference)
   Set Visibility(
    execute=[@n10.then, @n8.else]
    Visibility=Select(Option 0=Collapsed, Option 1=SelfHitTestInvisible, Index=AND Boolean!(A=@n12, B=@n13, C:bool=@n11) -> (ReturnValue))
    self=@n14
   )
   > ShowLoopLimitMessage(self:WBP Jam Now Playing Status Message Object Reference=@n14)
   n15 = Get IsViewingPlayerAJammer_Local()
   n16 = Get StatusMessage()
   SetVisibility(execute=@n10.else, self=@n16, InVisibility=Select(Option 0=Collapsed, Option 1=SelfHitTestInvisible, Index=NOT Boolean(A=@n15)))
   > n17 = GetJamSessionForActor!(execute, Actor:Actor Object Reference=GetOwningPlayerPawn!(self:User Widget Object Reference) -> (ReturnValue:Pawn Object Reference)) -> (then, ReturnValue:Jam Session Interface Interface)
   n18 = GetMusicManager(self=@n17)
   > ShowJamPrivacyMessage!(
    execute
    self:WBP Jam Now Playing Status Message Object Reference=@n16
    PrivacyLevel:Text=SelectText(
     A="LOCTEXT(\"Closed - You Can Join\")"
     B="LOCTEXT(\"Closed - You Can Not Join\")"
     bPickA=IsLocalPlayerAllowedToJoinClosedJam!(self:Jam Session Component Music Manager Object Reference=@n18) -> (ReturnValue:bool)
    )
   ) -> (then)
  ]}
]
variables [
 CondensedView {type=boolean, default=true}
 GameplayEventRouter {type=GameplayEventRouterComponent, default=null}
 JamGlobalStateVM {type=JamGlobalStateVM, default=null}
 JamLoopChangedEventHandle {type=AsyncAction_StartListeningToEvent, default=null}
 JamPlayspace {type=JamPlayspace, default=null}
 JamSession {default=None}
 KeyChangedEventHandle {type=AsyncAction_StartListeningToStatefulEvent, default=null}
 LastCheckedNotifierComp {type=ActorComponent_JamSessionNotifier, default=null}
 MaxNumJammers {type=integer, default=4}
 ModeChangedEventHandle {type=AsyncAction_StartListeningToStatefulEvent, default=null}
 ParticipantAddedEventHandle {type=AbilityAsync_WaitGameplayTagAdded, default=null}
 ParticipantRemovedEventHandle {type=AbilityAsync_WaitGameplayTagRemoved, default=null}
 PartSwitcherEventHandle {type=AsyncAction_StartListeningToEvent, default=null}
 Privacy Changed Event Handle {type=AsyncAction_StartListeningToStatefulEvent, default=null}
 TempoChangedEventHandle {type=AsyncAction_StartListeningToStatefulEvent, default=null}
]
widgets {
 info {inheritedWidgetCount=0, parentClass=&CommonUserWidget, rootWidgetClass=&Overlay, widgetCount=9}
 widgets [
  Overlay {widgetClassPath=&Overlay, widget=Overlay}
  PostBufferUpdate {
   widgetClassPath=/Script/UMG.PostBufferUpdate
   bIsVariable=true
   parent=Overlay
   widget=PostBufferUpdate
   properties {
    visibility=HitTestInvisible
    updateBufferInfos [{
      bufferToUpdate=ESlatePostRT_0
      postParamUpdater=&self:WidgetTree.PostBufferUpdate.PostBufferBlurUpdater_0
     }]
   }
   slot {class=OverlaySlot, properties {horizontalAlignment=HAlign_Fill, verticalAlignment=VAlign_Fill}}
  }
  SizeBox_1 {
   widgetClassPath=&SizeBox
   parent=Overlay
   widget=SizeBox_1
   properties {bOverride_WidthOverride=true, maxDesiredWidth=500, minDesiredWidth=450, widthOverride=450}
   slot {class=OverlaySlot, properties {verticalAlignment=VAlign_Bottom, padding=<ltrb 35 0 0 35>}}
  }
  VerticalBox_1 {widgetClassPath=&VerticalBox, parent=SizeBox_1, widget=VerticalBox_1, slot {class=SizeBoxSlot}}
  StatusMessage {
   widgetClassPath=&/WBP_Jam_NowPlaying_StatusMessage
   bIsVariable=true
   parent=VerticalBox_1
   widget=StatusMessage
   properties {visibility=Collapsed}
   slot {class=VerticalBoxSlot, properties {padding=<ltrb 0 0 0 2>, size {sizeRule=Fill, value=1}}}
  }
  WBP_JamAudioStateDisplay {
   widgetClassPath=&/WBP_JamAudioStateDisplay
   bIsVariable=true
   parent=VerticalBox_1
   widget=WBP_JamAudioStateDisplay
   slot {class=VerticalBoxSlot, properties {padding=<ltrb 0 2 0 2>}}
  }
  WBP_PersonalJamLoopDisplay {
   widgetClassPath=&/WBP_PersonalJamLoopDisplay
   bIsVariable=true
   parent=VerticalBox_1
   widget=WBP_PersonalJamLoopDisplay
   properties {visibility=Collapsed}
   slot {class=VerticalBoxSlot}
  }
  JamLoopDisplayList {
   widgetClassPath=/Script/UMG.ListView
   bIsVariable=true
   parent=VerticalBox_1
   widget=JamLoopDisplayList
   properties {
    entryWidgetClass=&/WBP_JamLoopDisplay.WBP_JamLoopDisplay_C
    horizontalEntrySpacing=4
    numDesignerPreviewEntries=4
    verticalEntrySpacing=4
    visibility=SelfHitTestInvisible
   }
   slot {class=VerticalBoxSlot, properties {padding=<ltrb 0 2 0 2>}}
  }
  CommonListView_53 {
   widgetClassPath=&CommonListView
   bIsVariable=true
   parent=VerticalBox_1
   widget=CommonListView_53
   properties {
    clipping=Inherit
    entryWidgetClass=&WBP_JamLoopDisplay_C
    horizontalEntrySpacing=4
    numDesignerPreviewEntries=4
    verticalEntrySpacing=4
   }
   slot {class=VerticalBoxSlot}
  }
 ]
}
