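# Emote wheel overlay for the Jam system, hosting a centered part switcher and an expanded song info panel in a bottom-left safe zone.
# On activation it finds the owning player's jam session, caches its Gameplay Event Router and broadcasts that the part switcher is displayed.
# On deactivation it broadcasts the hidden state through the cached router, if that is still valid.
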WidgetBlueprint /FMJamSystem/Widgets/PartSwitcher/BP_JamEmoteWheelOverlay
graphs [EventGraph {nodes [
   Event On Deactivated()
   > n1 = Get!(execute:Exec) -> (then:Exec, else:Exec, EventRouter:Gameplay Event Router Component Object Reference)
   > Broadcast Event!(
    execute
    self:Gameplay Event Router Component Object Reference=@n1
    EventData:Jam Event Part Switcher Displayed Structure=Make Jam Event Part Switcher Displayed!(isDisplayed_1_#1:bool=False) -> (JamEvent_PartSwitcher_Displayed:Jam Event Part Switcher Displayed Structure) {is=title K2Node_MakeStruct}
    EventContext:Object Reference
   ) -> (then, ReturnValue:bool)
   Event On Activated()
   > n2 = GetJamSessionForActor!(execute, Actor:Actor Object Reference=GetOwningPlayer()) -> (then, ReturnValue:Jam Session Interface Interface)
   > Branch(Condition=IsValidInterface!(Interface:"Interface Interface (by ref)"=@n2) -> (ReturnValue:bool))
   n3 = GetGameplayEventRouterComponent!(self:Jam Session Interface Interface=@n2) -> (ReturnValue:Gameplay Event Router Component Object Reference)
   > Is Valid(InputObject=@n3)
   > n4 = Set EventRouter(EventRouter:Gameplay Event Router Component Object Reference=@n3) -> (Output_Get:Gameplay Event Router Component Object Reference)
   > Broadcast Event(self=@n4, EventData=Make Jam Event Part Switcher Displayed(isDisplayed_1_#1=true))
  ]}]
variables [EventRouter {type=GameplayEventRouterComponent, default=null}]
widgets {
 info {inheritedWidgetCount=0, parentClass=/Script/FMJamUI.JamEmoteWheelOverlay, rootWidgetClass=&CanvasPanel, widgetCount=4}
 widgets [
  CanvasPanel_0 {widgetClassPath=&CanvasPanel, widget=CanvasPanel_0}
  BP_JamPartSwitcher {
   widgetClassPath=&/BP_JamPartSwitcher
   bIsVariable=true
   parent=CanvasPanel_0
   widget=BP_JamPartSwitcher
   slot {
    class=CanvasPanelSlot
    properties {layoutData {alignment=<0.5 0.5>, anchors {maximum=<0.5 0.5>, minimum=<0.5 0.5>}, offsets=<ltrb 0 0 352 352>}}
   }
  }
  SafeZone_0 {
   widgetClassPath=&SafeZone
   bIsVariable=true
   parent=CanvasPanel_0
   widget=SafeZone_0
   slot {
    class=CanvasPanelSlot
    properties {bAutoSize=true, layoutData {alignment=<0 1>, anchors {maximum=<0 1>, minimum=<0 1>}, offsets=<ltrb 0 0 100 30>}}
   }
  }
  WBP_JamSongInfo {
   widgetClassPath=/FMJamSystem/Widgets/SongInfo/WBP_JamSongInfo
   bIsVariable=true
   parent=SafeZone_0
   widget=WBP_JamSongInfo
   properties {condensedView=false}
   slot {class=SafeZoneSlot}
  }
 ]
}
