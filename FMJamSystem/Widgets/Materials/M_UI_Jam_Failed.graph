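# UI translucent material for the Jam system's failed indicator, built from a red-tinted alert icon texture parameter.
# The texture sample's RGB feeds the UVs input of ScaleUVsByCenter, whose result MF_Jam_AntiAliasing thresholds into a smoothed edge mask.
# That mask is multiplied by a red constant for color and, separately, added to a dimmed copy of the scale function's 0-1 mask.
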Material /FMJamSystem/Widgets/Materials/M_UI_Jam_Failed
graph {
 properties {bAutomaticallySetUsageInEditor=false, blendMode=BLEND_Translucent, materialDomain=MD_UI, usedShadingModels=Default Lit}
 nodes [
  n1 = MaterialFunctionCall(
   UVs=TextureSampleParameter2D() {parameterName=Texture, texture=/FMJamSystem/Widgets/Textures/T_UI_JamAlertIcon}.RGB
   Texture Scale=Constant() {r=0.9}
  ) {materialFunction=&ScaleUVsByCenter}
  n2 = MaterialFunctionCall!(Value=@n1.UVs, Threshold=Constant() {r=0.8}, Smoothing) {materialFunction=&/MF_Jam_AntiAliasing}
  Add(A=@n2.EitherSide, B=Multiply(A=@n1.0-1 mask) {constB=0.8})
  Multiply(A=Constant3Vector() {constant=<rgba 1 0.147 0.1329 1>}, B=@n2.EitherSide)
 ]
}
