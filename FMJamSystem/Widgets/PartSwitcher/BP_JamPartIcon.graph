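# Common activatable widget for a jam part-switcher icon, showing a loop type and its resolve status.
# Functions set the dynamic icon material's symbol index and two colors by loop type.
# Status changes clear a timer, then play mutually exclusive animations: Pending waits one second, Succeeded is immediate, Failed and Flagged_Suppressed share the failed look.
# An invisible button, limited by platform and input visibility, broadcasts On Button Touched.
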WidgetBlueprint /FMJamSystem/Widgets/PartSwitcher/BP_JamPartIcon
animations [
 PendingAnim {
  binding_count=3
  length_seconds=1
  object=&self:PendingAnim
  sequence {bindings [
    Pending_Status {
     id=#1
     possessed_class=Image
     tracks [
      {
       type=MovieSceneFloatTrack
       display_name=Render Opacity
       sections [{type=MovieSceneFloatSection, channels [None {type=MovieSceneScriptingFloatChannel, keys [[6,0],[12,1]]}]}]
      }
      {
       type=MovieScene2DTransformTrack
       sections [{
         type=MovieScene2DTransformSection
         channels [
          Scale.X {type=MovieSceneScriptingFloatChannel, const [1.3,1]}
          Scale.Y {type=MovieSceneScriptingFloatChannel, const [1.3,1]}
         ]
        }]
      }
     ]
    }
    Image_Icon {
     id=#2
     possessed_class=Image
     tracks [{
       type=MovieSceneFloatTrack
       display_name=Render Opacity
       sections [{type=MovieSceneFloatSection, channels [None {type=MovieSceneScriptingFloatChannel, keys [[0,1],[6,0.4]]}]}]
      }]
    }
    Failed_Status {
     id=#3
     possessed_class=Image
     tracks [{
       type=MovieSceneFloatTrack
       display_name=Render Opacity
       sections [{type=MovieSceneFloatSection, channels [None {type=MovieSceneScriptingFloatChannel, const [0,1]}]}]
      }]
    }
   ]}
 }
 FailedAnim {
  binding_count=3
  length_seconds=1
  object=&self:FailedAnim
  sequence {bindings [
    Failed_Status {
     id=#4
     possessed_class=Image
     tracks [
      {
       type=MovieSceneFloatTrack
       display_name=Render Opacity
       sections [{type=MovieSceneFloatSection, channels [None {type=MovieSceneScriptingFloatChannel, const [1,1]}]}]
      }
      {
       type=MovieScene2DTransformTrack
       sections [
        {
         type=MovieScene2DTransformSection
         channels [
          Scale.X {type=MovieSceneScriptingFloatChannel, keys [[0,0],[12,1],[20,0.9]]}
          Scale.Y {type=MovieSceneScriptingFloatChannel, keys [[0,0],[12,1],[20,0.9]]}
         ]
        }
        {
         type=MovieScene2DTransformSection
         channels [
          Translation.X {type=MovieSceneScriptingFloatChannel, const [0,1]}
          Translation.Y {type=MovieSceneScriptingFloatChannel, keys [[0,-2],[20,-2.4]]}
         ]
        }
       ]
      }
     ]
    }
    Pending_Status {
     id=#5
     possessed_class=Image
     tracks [{
       type=MovieSceneFloatTrack
       display_name=Render Opacity
       sections [{type=MovieSceneFloatSection, channels [None {type=MovieSceneScriptingFloatChannel, const [0,1]}]}]
      }]
    }
    Image_Icon {
     id=#6
     possessed_class=Image
     tracks [{
       type=MovieSceneFloatTrack
       display_name=Render Opacity
       sections [{type=MovieSceneFloatSection, channels [None {type=MovieSceneScriptingFloatChannel, keys [[0,1],[12,0.4]]}]}]
      }]
    }
   ]}
 }
 SucceededAnim {
  binding_count=3
  length_seconds=0.45
  object=&self:SucceededAnim
  sequence {bindings [
    Pending_Status {
     id=#7
     possessed_class=Image
     tracks [{
       type=MovieSceneFloatTrack
       display_name=Render Opacity
       sections [{type=MovieSceneFloatSection, channels [None {type=MovieSceneScriptingFloatChannel, const [0,1]}]}]
      }]
    }
    Image_Icon {
     id=#8
     possessed_class=Image
     tracks [{
       type=MovieSceneFloatTrack
       display_name=Render Opacity
       sections [{type=MovieSceneFloatSection, channels [None {type=MovieSceneScriptingFloatChannel, const [1,1]}]}]
      }]
    }
    Failed_Status {
     id=#9
     possessed_class=Image
     tracks [{
       type=MovieSceneFloatTrack
       display_name=Render Opacity
       sections [{type=MovieSceneFloatSection, channels [None {type=MovieSceneScriptingFloatChannel, const [0,1]}]}]
      }]
    }
   ]}
 }
]
graphs [
 EventGraph {nodes [
   "On Clicked (Touch_Button)"() {is=title K2Node_ComponentBoundEvent}
   > Call On Button Touched!(execute:Exec, self:BP Jam Part Icon Object Reference) -> (then:Exec) {is=title K2Node_CallDelegate}
   n1 = Process New Resolve Status!() -> (OutputDelegate, then, New Resolve State:EJamMusicSlotVMResolveState Enum=Unknown) {is=title K2Node_CustomEvent}
   > Clear and Invalidate Timer by Handle(Handle=Get Start Pending Appearance Handle() -> (Start Pending Appearance Handle:Timer Handle Structure))
   > n2 = Switch on EJamMusicSlotVMResolveState!(execute, Selection:EJamMusicSlotVMResolveState Enum=@n1.New Resolve State) -> (Unknown:Exec, Pending:Exec, Failed:Exec, Succeeded:Exec, Flagged_Suppressed:Exec) {is=title K2Node_SwitchEnum}
   n3 = Set Timer by Event(execute=@n2.Pending, Delegate=Create Event(), Time=1, bLooping=false)
   > Set Start Pending Appearance Handle(Start Pending Appearance Handle:Timer Handle Structure=@n3) -> (Output_Get:Timer Handle Structure)
   StartFailedAnimation(execute=@n2.Failed)
   StartFailedAnimation(execute=@n2.Flagged_Suppressed)
   StartSucceededAnimation(execute=@n2.Succeeded)
   Start Pending Appearance() {is=title K2Node_CustomEvent}
   > StartPendingAnimatin()
   Event PreConstruct()
   > SetBrushFromMaterial(self=Get Image_Icon() -> (Image_Icon:Image Object Reference), Material=Get Gradient() -> (Gradient:Material Instance Object Reference))
   > SetDesiredSizeOverride!(execute, self:Image Object Reference=Get Image_Icon(), DesiredSize:Vector 2D Structure=Get ImageSize() -> (ImageSize:Vector 2D Structure)) -> (then)
  ]}
 On Button Touched {nodes [On Button Touched() {is=title K2Node_FunctionEntry}]}
 Set Active Icon by LoopType {nodes [
   n1 = Set Active Icon by LoopType!() -> (then, LoopType:EFMJamLoopType Enum=Num) {is=title K2Node_FunctionEntry}
   > n2 = Sequence()
   > n3 = Set LoopType(LoopType:EFMJamLoopType Enum=@n1) -> (Output_Get:EFMJamLoopType Enum=Lead)
   > n4 = Switch on EFMJamLoopType!(execute, Selection:EFMJamLoopType Enum=@n3) -> (Lead:Exec, Misc:Exec, Bass:Exec, Beat:Exec, Num:Exec) {is=title K2Node_SwitchEnum}
   Set TrackTypeValue(execute=@n4.Bass, TrackTypeValue:float=1) -> (Output_Get:float)
   Set TrackTypeValue(execute=@n4.Beat, TrackTypeValue=3)
   Set TrackTypeValue(execute=@n4.Lead, TrackTypeValue=2)
   Set TrackTypeValue(execute=@n4.Misc, TrackTypeValue=0)
   Set TrackTypeValue(execute=@n4.Num, TrackTypeValue=5)
   n5 = GetDynamicMaterial(execute=@n2.then_1, self=Get Image_Icon())
   > SetScalarParameterValue(self=@n5, ParameterName=Icon, Value=Get TrackTypeValue() -> (TrackTypeValue:float))
  ]}
 SetFocusIconColor {nodes [
   n1 = SetFocusIconColor!() -> (then, InColorAndOpacity:Linear Color Structure) {is=title K2Node_FunctionEntry}
   > SetColorAndOpacity(self:Image Object Reference=Get Image_Icon(), InColorAndOpacity:Linear Color Structure=@n1)
  ]}
 SetIconColorByJamLoopType {nodes [
   n1 = SetIconColorByJamLoopType!() -> (then, JamLoopType:EFMJamLoopType Enum=Lead) {is=title K2Node_FunctionEntry}
   > n2 = Sequence()
   > n3 = Switch on EFMJamLoopType(Selection=@n1)
   Set Color_1(execute=@n3.Bass, Color_1:Linear Color Structure=(R=0.1878,G=1,B=0.1356,A=1)) -> (Output_Get:Linear Color Structure)
   > Set Color_2(Color_2:Linear Color Structure=(R=0.183,G=0.8646,B=0.1318,A=1)) -> (Output_Get:Linear Color Structure)
   Set Color_1(execute=@n3.Beat, Color_1=(R=0,G=0.552,B=0.8879,A=1))
   > Set Color_2(Color_2=(R=0,G=0.5019,B=0.8073,A=1))
   Set Color_1(execute=@n3.Lead, Color_1=(R=1,G=0.7605,B=0.0356,A=1))
   > Set Color_2(Color_2=(R=0.9375,G=0.713,B=0.0334,A=1))
   Set Color_1(execute=@n3.Misc, Color_1=(R=0.9473,G=0.1248,B=0.4678,A=1))
   > Set Color_2(Color_2=(R=0.9133,G=0.0702,B=0.9473,A=1))
   Set Color_1(execute=@n3.Num, Color_1=(R=0.8879,G=0.956,B=0.159,A=1))
   > Set Color_2(Color_2=(R=0.956,G=0.7157,B=0.1413,A=1))
   n4 = GetDynamicMaterial(execute=@n2.then_1, self=Get Image_Icon())
   > n5 = Set Icon_MaterialInstance(Icon_MaterialInstance:Material Instance Dynamic Object Reference=@n4) -> (Output_Get:Material Instance Dynamic Object Reference)
   > SetVectorParameterValue(self=@n5, ParameterName=QM_CL1, Value=Get Color_1() -> (Color_1:Linear Color Structure))
   > SetVectorParameterValue(self=@n5, ParameterName=QM_CL2, Value=Get Color_2() -> (Color_2:Linear Color Structure))
  ]}
 StartFailedAnimation {nodes [
   StartFailedAnimation() {is=title K2Node_FunctionEntry}
   > StopAnimation(InAnimation=Get PendingAnim() -> (PendingAnim:Widget Animation Object Reference))
   > StopAnimation(InAnimation=Get SucceededAnim() -> (SucceededAnim:Widget Animation Object Reference))
   > PlayAnimation(InAnimation=Get FailedAnim() -> (FailedAnim:Widget Animation Object Reference), PlaybackSpeed=1.5)
  ]}
 StartPendingAnimatin {nodes [
   StartPendingAnimatin() {is=title K2Node_FunctionEntry}
   > StopAnimation(InAnimation=Get SucceededAnim())
   > StopAnimation(InAnimation=Get FailedAnim())
   > PlayAnimation(InAnimation=Get PendingAnim(), PlaybackSpeed=1.5)
  ]}
 StartSucceededAnimation {nodes [
   StartSucceededAnimation() {is=title K2Node_FunctionEntry}
   > StopAnimation(InAnimation=Get PendingAnim())
   > StopAnimation(InAnimation=Get FailedAnim())
   > PlayAnimation(InAnimation=Get SucceededAnim())
  ]}
]
variables [
 Gradient {type=MaterialInstance, default=/FMJamSystem/Materials/MI_UI_Icon_Instrument_Type}
 ImageSize {type=Vector2D, default=<32 32>}
 LoopType {type=EFMJamLoopType, default=Lead}
 On Button Touched {}
 Start Pending Appearance Handle {type=TimerHandle}
]
widgets {
 info {inheritedWidgetCount=0, parentClass=&CommonActivatableWidget, rootWidgetClass=&Overlay, widgetCount=6}
 widgets [
  Overlay_85 {widgetClassPath=&Overlay, widget=Overlay_85}
  CommonVisibilityWidgetLegacy_0 {
   widgetClassPath=&CommonVisibilityWidgetLegacy
   parent=Overlay_85
   widget=CommonVisibilityWidgetLegacy_0
   properties {
    bShowForGamepad=false
    bShowForMac=false
    bShowForMouseAndKeyboard=false
    bShowForPS4=false
    bShowForPS5=false
    bShowForXBox=false
    bShowForXSX=false
    padding=<ltrb 0 0 0 0>
   }
   slot {class=OverlaySlot, properties {horizontalAlignment=HAlign_Fill, verticalAlignment=VAlign_Fill}}
  }
  Touch_Button {
   widgetClassPath=&Button
   parent=CommonVisibilityWidgetLegacy_0
   widget=Touch_Button
   properties {backgroundColor=<rgba 1 1 1 0>, colorAndOpacity=<rgba 1 1 1 0>}
   slot {class=BorderSlot, properties {padding=<ltrb 0 0 0 0>}}
  }
  Pending_Status {
   widgetClassPath=&Image
   bIsVariable=true
   parent=Overlay_85
   widget=Pending_Status
   properties {
    renderOpacity=0
    visibility=SelfHitTestInvisible
    brush {
     bIsDynamicallyLoaded=false
     drawAs=Image
     imageType=FullColor
     mirroring=NoMirror
     resourceName=None
     resourceObject=/FMJamSystem/Widgets/Materials/MI_UI_Jam_Pending
     tiling=NoTile
     imageSize=<32 32>
     margin=<ltrb 0 0 0 0>
     outlineSettings {
      bUseBrushTransparency=false
      roundingType=HalfHeightRadius
      width=0
      color {colorUseRule=UseColor_Specified, specifiedColor=<rgba 0 0 0 0>}
      cornerRadii=<0 0 0 0>
     }
     tintColor {colorUseRule=UseColor_Specified, specifiedColor=<rgba 1 1 1 1>}
     uVRegion {bIsValid=false, max=<0 0>, min=<0 0>}
    }
   }
   slot {class=OverlaySlot, properties {horizontalAlignment=HAlign_Fill, verticalAlignment=VAlign_Fill}}
  }
  Image_Icon {
   widgetClassPath=&Image
   bIsVariable=true
   parent=Overlay_85
   widget=Image_Icon
   properties {
    visibility=SelfHitTestInvisible
    brush {
     bIsDynamicallyLoaded=false
     drawAs=Image
     imageType=NoImage
     mirroring=NoMirror
     resourceName=None
     resourceObject=&MI_UI_Icon_Instrument_Type
     tiling=NoTile
     imageSize=<32 32>
     margin=<ltrb 0 0 0 0>
     outlineSettings {
      bUseBrushTransparency=false
      roundingType=HalfHeightRadius
      width=0
      color {colorUseRule=UseColor_Specified, specifiedColor=<rgba 0 0 0 0>}
      cornerRadii=<0 0 0 0>
     }
     tintColor {colorUseRule=UseColor_Specified, specifiedColor=<rgba 1 1 1 1>}
     uVRegion {bIsValid=false, max=<0 0>, min=<0 0>}
    }
   }
   slot {class=OverlaySlot, properties {horizontalAlignment=HAlign_Fill, verticalAlignment=VAlign_Fill}}
  }
  Failed_Status {
   widgetClassPath=&Image
   bIsVariable=true
   parent=Overlay_85
   widget=Failed_Status
   properties {
    renderOpacity=0
    visibility=SelfHitTestInvisible
    brush {
     bIsDynamicallyLoaded=false
     drawAs=Image
     imageType=FullColor
     mirroring=NoMirror
     resourceName=None
     resourceObject=/FMJamSystem/Widgets/Materials/M_UI_Jam_Failed
     tiling=NoTile
     imageSize=<32 32>
     margin=<ltrb 0 0 0 0>
     outlineSettings {
      bUseBrushTransparency=false
      roundingType=HalfHeightRadius
      width=0
      color {colorUseRule=UseColor_Specified, specifiedColor=<rgba 0 0 0 0>}
      cornerRadii=<0 0 0 0>
     }
     tintColor {colorUseRule=UseColor_Specified, specifiedColor=<rgba 1 1 1 1>}
     uVRegion {bIsValid=false, max=<0 0>, min=<0 0>}
    }
   }
   slot {class=OverlaySlot, properties {horizontalAlignment=HAlign_Fill, verticalAlignment=VAlign_Fill}}
  }
 ]
}
