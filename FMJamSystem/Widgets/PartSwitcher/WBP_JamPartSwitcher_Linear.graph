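# Linear carousel widget that cycles Jam part (loop type) icons, with "All" and legacy layouts chosen by IsPartEnabled.
# Clicks and icon touches increment the view model's looped MultiEmote index, refresh visuals, play a slide animation and report a Flowrider interaction.
# Index-to-loop-type lookups query the player's emote options component, falling back to LoopTypeList.
# Button background follows input type; non-mobile Tick toggles button hit-testing using circular mouse-distance checks.
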WidgetBlueprint /FMJamSystem/Widgets/PartSwitcher/WBP_JamPartSwitcher_Linear
animations [
 A_Cycle_Back_All {
  binding_count=10
  length_seconds=0.25
  object=&self:A_Cycle_Back_All
  sequence {
   bindings [
    PartIconMain {
     id=#1
     possessed_class=BP_JamPartIcon_C
     tracks [
      {
       type=MovieSceneColorTrack
       display_name=Color and Opacity
       sections [{
         type=MovieSceneColorSection
         end_frame=60
         start_frame=0
         channels [
          Color.R {type=MovieSceneScriptingFloatChannel}
          Color.G {type=MovieSceneScriptingFloatChannel}
          Color.B {type=MovieSceneScriptingFloatChannel}
          Color.A {type=MovieSceneScriptingFloatChannel, keys [[0,0.6],[60,1]]}
         ]
        }]
      }
      {
       type=MovieScene2DTransformTrack
       sections [{
         type=MovieScene2DTransformSection
         end_frame=60
         start_frame=0
         channels [
          Translation.X {type=MovieSceneScriptingFloatChannel}
          Translation.Y {type=MovieSceneScriptingFloatChannel}
          Angle {type=MovieSceneScriptingFloatChannel}
          Scale.X {type=MovieSceneScriptingFloatChannel, keys [[0,1.15],[60,2.65]]}
          Scale.Y {type=MovieSceneScriptingFloatChannel, keys [[0,1.15],[60,2.65]]}
          Shear.X {type=MovieSceneScriptingFloatChannel}
          Shear.Y {type=MovieSceneScriptingFloatChannel}
         ]
        }]
      }
     ]
    }
    CanvasPanelSlot_5 {
     id=#2
     parent_id=#3
     possessed_class=CanvasPanelSlot
     tracks [{
       type=MovieSceneMarginTrack
       display_name="Offsets (Layout Data)"
       sections [{
         type=MovieSceneMarginSection
         end_frame=60
         start_frame=0
         channels [
          Left {type=MovieSceneScriptingFloatChannel, keys [[0,120],[60,78]]}
          Top {type=MovieSceneScriptingFloatChannel, default=0}
          Right {type=MovieSceneScriptingFloatChannel, default=0}
          Bottom {type=MovieSceneScriptingFloatChannel, default=0}
         ]
        }]
      }]
    }
    CanvasPanelSlot_2 {
     id=#4
     parent_id=#5
     possessed_class=CanvasPanelSlot
     tracks [{
       type=MovieSceneMarginTrack
       display_name="Offsets (Layout Data)"
       sections [{
         type=MovieSceneMarginSection
         end_frame=60
         start_frame=0
         channels [
          Left {type=MovieSceneScriptingFloatChannel, keys [[0,0],[60,-78]]}
          Top {type=MovieSceneScriptingFloatChannel, default=0}
          Right {type=MovieSceneScriptingFloatChannel, default=0}
          Bottom {type=MovieSceneScriptingFloatChannel, default=0}
         ]
        }]
      }]
    }
    PartIconNext1 {
     id=#6
     possessed_class=BP_JamPartIcon_C
     tracks [
      {
       type=MovieScene2DTransformTrack
       sections [{
         type=MovieScene2DTransformSection
         end_frame=60
         start_frame=0
         channels [
          Translation.X {type=MovieSceneScriptingFloatChannel, const [0,1]}
          Translation.Y {type=MovieSceneScriptingFloatChannel, const [0,1]}
          Angle {type=MovieSceneScriptingFloatChannel, const [0,1]}
          Scale.X {type=MovieSceneScriptingFloatChannel, keys [[0,1],[60,0.75]]}
          Scale.Y {type=MovieSceneScriptingFloatChannel, keys [[0,1],[60,0.75]]}
          Shear.X {type=MovieSceneScriptingFloatChannel, const [0,1]}
          Shear.Y {type=MovieSceneScriptingFloatChannel, const [0,1]}
         ]
        }]
      }
      {
       type=MovieSceneColorTrack
       display_name=Color and Opacity
       sections [{
         type=MovieSceneColorSection
         end_frame=60
         start_frame=0
         channels [
          Color.R {type=MovieSceneScriptingFloatChannel, const [1,1]}
          Color.G {type=MovieSceneScriptingFloatChannel, const [1,1]}
          Color.B {type=MovieSceneScriptingFloatChannel, const [1,1]}
          Color.A {type=MovieSceneScriptingFloatChannel, keys [[0,0.6],[60,0.25]]}
         ]
        }]
      }
     ]
    }
    CanvasPanelSlot_7 {
     id=#7
     parent_id=#1
     possessed_class=CanvasPanelSlot
     tracks [{
       type=MovieSceneMarginTrack
       display_name="Offsets (Layout Data)"
       sections [{
         type=MovieSceneMarginSection
         end_frame=60
         start_frame=0
         channels [
          Left {type=MovieSceneScriptingFloatChannel, keys [[0,78],[60,0]]}
          Top {type=MovieSceneScriptingFloatChannel, default=0}
          Right {type=MovieSceneScriptingFloatChannel, default=0}
          Bottom {type=MovieSceneScriptingFloatChannel, default=0}
         ]
        }]
      }]
    }
    CanvasPanelSlot_6 {
     id=#8
     parent_id=#6
     possessed_class=CanvasPanelSlot
     tracks [{
       type=MovieSceneMarginTrack
       display_name="Offsets (Layout Data)"
       sections [{
         type=MovieSceneMarginSection
         end_frame=60
         start_frame=0
         channels [
          Left {type=MovieSceneScriptingFloatChannel, keys [[0,-78],[60,-120]]}
          Top {type=MovieSceneScriptingFloatChannel, default=0}
          Right {type=MovieSceneScriptingFloatChannel, default=0}
          Bottom {type=MovieSceneScriptingFloatChannel, default=0}
         ]
        }]
      }]
    }
    PartIconNext2 {
     id=#5
     possessed_class=BP_JamPartIcon_C
     tracks [
      {
       type=MovieScene2DTransformTrack
       sections [{
         type=MovieScene2DTransformSection
         end_frame=60
         start_frame=0
         channels [
          Translation.X {type=MovieSceneScriptingFloatChannel}
          Translation.Y {type=MovieSceneScriptingFloatChannel}
          Angle {type=MovieSceneScriptingFloatChannel}
          Scale.X {type=MovieSceneScriptingFloatChannel, keys [[0,3],[60,1.3]]}
          Scale.Y {type=MovieSceneScriptingFloatChannel, keys [[0,3],[60,1.3]]}
          Shear.X {type=MovieSceneScriptingFloatChannel}
          Shear.Y {type=MovieSceneScriptingFloatChannel}
         ]
        }]
      }
      {
       type=MovieSceneColorTrack
       display_name=Color and Opacity
       sections [{
         type=MovieSceneColorSection
         end_frame=60
         start_frame=0
         channels [
          Color.R {type=MovieSceneScriptingFloatChannel}
          Color.G {type=MovieSceneScriptingFloatChannel}
          Color.B {type=MovieSceneScriptingFloatChannel}
          Color.A {type=MovieSceneScriptingFloatChannel, keys [[0,1],[60,0.6]]}
         ]
        }]
      }
     ]
    }
    PartIconPrevious2 {
     id=#9
     possessed_class=BP_JamPartIcon_C
     tracks [{
       type=MovieSceneColorTrack
       display_name=Color and Opacity
       sections [{
         type=MovieSceneColorSection
         end_frame=60
         start_frame=0
         channels [
          Color.R {type=MovieSceneScriptingFloatChannel}
          Color.G {type=MovieSceneScriptingFloatChannel}
          Color.B {type=MovieSceneScriptingFloatChannel}
          Color.A {type=MovieSceneScriptingFloatChannel, keys [[0,0.25],[29,0],[30,0],[60,0.25]]}
         ]
        }]
      }]
    }
    CanvasPanelSlot_4 {
     id=#10
     parent_id=#9
     possessed_class=CanvasPanelSlot
     tracks [{
       type=MovieSceneMarginTrack
       display_name="Offsets (Layout Data)"
       sections [{
         type=MovieSceneMarginSection
         end_frame=60
         start_frame=0
         channels [
          Left {type=MovieSceneScriptingFloatChannel, keys [[0,-120],[29,-136],[30,136],[60,120]]}
          Top {type=MovieSceneScriptingFloatChannel, const [0,2]}
          Right {type=MovieSceneScriptingFloatChannel, const [0,2]}
          Bottom {type=MovieSceneScriptingFloatChannel, const [0,2]}
         ]
        }]
      }]
    }
    PartIconPrevious1 {
     id=#3
     possessed_class=BP_JamPartIcon_C
     tracks [
      {
       type=MovieScene2DTransformTrack
       sections [{
         type=MovieScene2DTransformSection
         end_frame=60
         start_frame=0
         channels [
          Translation.X {type=MovieSceneScriptingFloatChannel, const [0,1]}
          Translation.Y {type=MovieSceneScriptingFloatChannel, const [0,1]}
          Angle {type=MovieSceneScriptingFloatChannel, const [0,1]}
          Scale.X {type=MovieSceneScriptingFloatChannel, keys [[0,0.75],[60,1]]}
          Scale.Y {type=MovieSceneScriptingFloatChannel, keys [[0,0.75],[60,1]]}
          Shear.X {type=MovieSceneScriptingFloatChannel, const [0,1]}
          Shear.Y {type=MovieSceneScriptingFloatChannel, const [0,1]}
         ]
        }]
      }
      {
       type=MovieSceneColorTrack
       display_name=Color and Opacity
       sections [{
         type=MovieSceneColorSection
         end_frame=60
         start_frame=0
         channels [
          Color.R {type=MovieSceneScriptingFloatChannel, const [1,1]}
          Color.G {type=MovieSceneScriptingFloatChannel, const [1,1]}
          Color.B {type=MovieSceneScriptingFloatChannel, const [1,1]}
          Color.A {type=MovieSceneScriptingFloatChannel, keys [[0,0.25],[60,0.6]]}
         ]
        }]
      }
     ]
    }
   ]
   display_rate {numerator=240}
  }
 }
 A_Cycle_Forward_All {
  binding_count=10
  length_seconds=0.25
  object=&self:A_Cycle_Forward_All
  sequence {
   bindings [
    PartIconMain {
     id=#1
     possessed_class=BP_JamPartIcon_C
     tracks [
      {
       type=MovieScene2DTransformTrack
       sections [{
         type=MovieScene2DTransformSection
         end_frame=60
         start_frame=0
         channels [
          Translation.X {type=MovieSceneScriptingFloatChannel}
          Translation.Y {type=MovieSceneScriptingFloatChannel}
          Angle {type=MovieSceneScriptingFloatChannel}
          Scale.X {type=MovieSceneScriptingFloatChannel, keys [[0,1],[60,2.65]]}
          Scale.Y {type=MovieSceneScriptingFloatChannel, keys [[0,1],[60,2.65]]}
          Shear.X {type=MovieSceneScriptingFloatChannel}
          Shear.Y {type=MovieSceneScriptingFloatChannel}
         ]
        }]
      }
      {
       type=MovieSceneColorTrack
       display_name=Color and Opacity
       sections [{
         type=MovieSceneColorSection
         end_frame=60
         start_frame=0
         channels [
          Color.R {type=MovieSceneScriptingFloatChannel}
          Color.G {type=MovieSceneScriptingFloatChannel}
          Color.B {type=MovieSceneScriptingFloatChannel}
          Color.A {type=MovieSceneScriptingFloatChannel, keys [[0,0.6],[60,1]]}
         ]
        }]
      }
     ]
    }
    CanvasPanelSlot_5 {
     id=#2
     parent_id=#3
     possessed_class=CanvasPanelSlot
     tracks [{
       type=MovieSceneMarginTrack
       display_name="Offsets (Layout Data)"
       sections [{
         type=MovieSceneMarginSection
         end_frame=60
         start_frame=0
         channels [
          Left {type=MovieSceneScriptingFloatChannel, keys [[0,0],[60,90]]}
          Top {type=MovieSceneScriptingFloatChannel, const [0,2]}
          Right {type=MovieSceneScriptingFloatChannel, const [0,2]}
          Bottom {type=MovieSceneScriptingFloatChannel, const [0,2]}
         ]
        }]
      }]
    }
    PartIconPrevious2 {id=#11, possessed_class=BP_JamPartIcon_C, tracks []}
    CanvasPanelSlot_2 {
     id=#4
     parent_id=#5
     possessed_class=CanvasPanelSlot
     tracks [{
       type=MovieSceneMarginTrack
       display_name="Offsets (Layout Data)"
       sections [{
         type=MovieSceneMarginSection
         end_frame=60
         start_frame=0
         channels [
          Left {type=MovieSceneScriptingFloatChannel, keys [[0,-144],[60,-90]]}
          Top {type=MovieSceneScriptingFloatChannel, default=0}
          Right {type=MovieSceneScriptingFloatChannel, default=0}
          Bottom {type=MovieSceneScriptingFloatChannel, default=0}
         ]
        }]
      }]
    }
    PartIconNext1 {
     id=#6
     possessed_class=BP_JamPartIcon_C
     tracks [{
       type=MovieSceneColorTrack
       display_name=Color and Opacity
       sections [{
         type=MovieSceneColorSection
         end_frame=60
         start_frame=0
         channels [
          Color.R {type=MovieSceneScriptingFloatChannel}
          Color.G {type=MovieSceneScriptingFloatChannel}
          Color.B {type=MovieSceneScriptingFloatChannel}
          Color.A {type=MovieSceneScriptingFloatChannel, keys [[0,0.6],[25,0],[26,0],[60,0.6]]}
         ]
        }]
      }]
    }
    CanvasPanelSlot_4 {
     id=#12
     parent_id=#11
     possessed_class=CanvasPanelSlot
     tracks [{
       type=MovieSceneMarginTrack
       display_name="Offsets (Layout Data)"
       sections [{
         type=MovieSceneMarginSection
         end_frame=60
         start_frame=0
         channels [
          Left {type=MovieSceneScriptingFloatChannel, keys [[0,90],[60,144]]}
          Top {type=MovieSceneScriptingFloatChannel, default=0}
          Right {type=MovieSceneScriptingFloatChannel, default=0}
          Bottom {type=MovieSceneScriptingFloatChannel, default=0}
         ]
        }]
      }]
    }
    CanvasPanelSlot_7 {
     id=#7
     parent_id=#1
     possessed_class=CanvasPanelSlot
     tracks [{
       type=MovieSceneMarginTrack
       display_name="Offsets (Layout Data)"
       sections [{
         type=MovieSceneMarginSection
         end_frame=60
         start_frame=0
         channels [
          Left {type=MovieSceneScriptingFloatChannel, keys [[0,-90],[60,0]]}
          Top {type=MovieSceneScriptingFloatChannel, default=0}
          Right {type=MovieSceneScriptingFloatChannel, default=0}
          Bottom {type=MovieSceneScriptingFloatChannel, default=0}
         ]
        }]
      }]
    }
    CanvasPanelSlot_6 {
     id=#8
     parent_id=#6
     possessed_class=CanvasPanelSlot
     tracks [{
       type=MovieSceneMarginTrack
       display_name="Offsets (Layout Data)"
       sections [{
         type=MovieSceneMarginSection
         end_frame=60
         start_frame=0
         channels [
          Left {type=MovieSceneScriptingFloatChannel, keys [[0,120],[29,170],[30,-170],[60,-144]]}
          Top {type=MovieSceneScriptingFloatChannel, default=0}
          Right {type=MovieSceneScriptingFloatChannel, default=0}
          Bottom {type=MovieSceneScriptingFloatChannel, default=0}
         ]
        }]
      }]
    }
    PartIconNext2 {id=#5, possessed_class=BP_JamPartIcon_C, tracks []}
    PartIconPrevious1 {
     id=#3
     possessed_class=BP_JamPartIcon_C
     tracks [
      {
       type=MovieSceneColorTrack
       display_name=Color and Opacity
       sections [{
         type=MovieSceneColorSection
         end_frame=60
         start_frame=0
         channels [
          Color.R {type=MovieSceneScriptingFloatChannel}
          Color.G {type=MovieSceneScriptingFloatChannel}
          Color.B {type=MovieSceneScriptingFloatChannel}
          Color.A {type=MovieSceneScriptingFloatChannel, keys [[0,1],[60,0.6]]}
         ]
        }]
      }
      {
       type=MovieScene2DTransformTrack
       sections [{
         type=MovieScene2DTransformSection
         end_frame=60
         start_frame=0
         channels [
          Translation.X {type=MovieSceneScriptingFloatChannel}
          Translation.Y {type=MovieSceneScriptingFloatChannel}
          Angle {type=MovieSceneScriptingFloatChannel}
          Scale.X {type=MovieSceneScriptingFloatChannel, keys [[0,2.65],[60,1]]}
          Scale.Y {type=MovieSceneScriptingFloatChannel, keys [[0,2.65],[60,1]]}
          Shear.X {type=MovieSceneScriptingFloatChannel}
          Shear.Y {type=MovieSceneScriptingFloatChannel}
         ]
        }]
      }
     ]
    }
   ]
   display_rate {numerator=240}
  }
 }
 A_ButtonClick_Pulse {
  binding_count=1
  length_seconds=0.2
  object=&self:A_ButtonClick_Pulse
  sequence {bindings [SelectedInstrument_Background {
     id=#13
     possessed_class=Image
     tracks [{
       type=MovieSceneWidgetMaterialTrack
       sections [{
         type=MovieSceneParameterSection
         end_frame=4
         start_frame=0
         channels [
          Circle Size {type=MovieSceneScriptingFloatChannel, keys [[0,0.355],[3,0.375],[4,0.355]]}
          Strip SizeR {type=MovieSceneScriptingFloatChannel, default=1}
          Strip SizeG {type=MovieSceneScriptingFloatChannel, keys [[0,0.18],[3,0.2],[4,0.18]]}
          Strip SizeB {type=MovieSceneScriptingFloatChannel, default=0}
          Strip SizeA {type=MovieSceneScriptingFloatChannel, default=1}
         ]
        }]
      }]
    }]}
 }
 A_Button_Hover {
  binding_count=2
  length_seconds=0.3
  object=&self:A_Button_Hover
  sequence {bindings [
    PartSetVisibilitySwitcher {
     id=#14
     possessed_class=CommonVisibilitySwitcher
     tracks [{
       type=MovieScene2DTransformTrack
       sections [{
         type=MovieScene2DTransformSection
         end_frame=6
         start_frame=0
         channels [
          Scale.X {type=MovieSceneScriptingFloatChannel, keys [[0,1],[3,1.0375]]}
          Scale.Y {type=MovieSceneScriptingFloatChannel, keys [[0,1],[3,1.0375]]}
         ]
        }]
      }]
    }
    SelectedInstrument_Background {
     id=#15
     possessed_class=Image
     tracks [{
       type=MovieScene2DTransformTrack
       sections [{
         type=MovieScene2DTransformSection
         end_frame=6
         start_frame=0
         channels [
          Scale.X {type=MovieSceneScriptingFloatChannel, keys [[0,1],[3,1.0375],[6,1.0375]]}
          Scale.Y {type=MovieSceneScriptingFloatChannel, keys [[0,1],[3,1.0375],[6,1.0375]]}
         ]
        }]
      }]
    }
   ]}
 }
 A_Cycle_Back_Legacy {
  binding_count=8
  length_seconds=0.25
  object=&self:A_Cycle_Back_Legacy
  sequence {
   bindings [
    PartIconMain_Legacy {
     id=#16
     possessed_class=BP_JamPartIcon_C
     tracks [
      {
       type=MovieScene2DTransformTrack
       sections [{
         type=MovieScene2DTransformSection
         end_frame=60
         start_frame=0
         channels [
          Translation.X {type=MovieSceneScriptingFloatChannel}
          Translation.Y {type=MovieSceneScriptingFloatChannel}
          Angle {type=MovieSceneScriptingFloatChannel}
          Scale.X {type=MovieSceneScriptingFloatChannel, keys [[0,1.15],[60,2.65]]}
          Scale.Y {type=MovieSceneScriptingFloatChannel, keys [[0,1.15],[60,2.65]]}
          Shear.X {type=MovieSceneScriptingFloatChannel}
          Shear.Y {type=MovieSceneScriptingFloatChannel}
         ]
        }]
      }
      {
       type=MovieSceneColorTrack
       display_name=Color and Opacity
       sections [{
         type=MovieSceneColorSection
         end_frame=60
         start_frame=0
         channels [
          Color.R {type=MovieSceneScriptingFloatChannel}
          Color.G {type=MovieSceneScriptingFloatChannel}
          Color.B {type=MovieSceneScriptingFloatChannel}
          Color.A {type=MovieSceneScriptingFloatChannel, keys [[0,0.6],[60,1]]}
         ]
        }]
      }
     ]
    }
    CanvasPanelSlot_1 {
     id=#17
     parent_id=#16
     possessed_class=CanvasPanelSlot
     tracks [{
       type=MovieSceneMarginTrack
       display_name="Offsets (Layout Data)"
       sections [{
         type=MovieSceneMarginSection
         end_frame=60
         start_frame=0
         channels [
          Left {type=MovieSceneScriptingFloatChannel, keys [[0,78],[60,0]]}
          Top {type=MovieSceneScriptingFloatChannel, default=0}
          Right {type=MovieSceneScriptingFloatChannel, default=0}
          Bottom {type=MovieSceneScriptingFloatChannel, default=0}
         ]
        }]
      }]
    }
    CanvasPanelSlot_2 {
     id=#18
     parent_id=#19
     possessed_class=CanvasPanelSlot
     tracks [{
       type=MovieSceneMarginTrack
       display_name="Offsets (Layout Data)"
       sections [{
         type=MovieSceneMarginSection
         end_frame=60
         start_frame=0
         channels [
          Left {type=MovieSceneScriptingFloatChannel, keys [[0,0],[60,-78]]}
          Top {type=MovieSceneScriptingFloatChannel, default=0}
          Right {type=MovieSceneScriptingFloatChannel, default=0}
          Bottom {type=MovieSceneScriptingFloatChannel, default=0}
         ]
        }]
      }]
    }
    CanvasPanelSlot_10 {
     id=#20
     parent_id=#21
     possessed_class=CanvasPanelSlot
     tracks [{
       type=MovieSceneMarginTrack
       display_name="Offsets (Layout Data)"
       sections [{
         type=MovieSceneMarginSection
         end_frame=60
         start_frame=0
         channels [
          Left {type=MovieSceneScriptingFloatChannel, keys [[0,120],[60,78]]}
          Top {type=MovieSceneScriptingFloatChannel, default=0}
          Right {type=MovieSceneScriptingFloatChannel, default=0}
          Bottom {type=MovieSceneScriptingFloatChannel, default=0}
         ]
        }]
      }]
    }
    PartIconPrevious_Legacy {
     id=#21
     possessed_class=BP_JamPartIcon_C
     tracks [
      {
       type=MovieScene2DTransformTrack
       sections [{
         type=MovieScene2DTransformSection
         end_frame=60
         start_frame=0
         channels [
          Translation.X {type=MovieSceneScriptingFloatChannel}
          Translation.Y {type=MovieSceneScriptingFloatChannel}
          Angle {type=MovieSceneScriptingFloatChannel}
          Scale.X {type=MovieSceneScriptingFloatChannel, keys [[0,0.75],[60,1]]}
          Scale.Y {type=MovieSceneScriptingFloatChannel, keys [[0,0.75],[60,1]]}
          Shear.X {type=MovieSceneScriptingFloatChannel}
          Shear.Y {type=MovieSceneScriptingFloatChannel}
         ]
        }]
      }
      {
       type=MovieSceneColorTrack
       display_name=Color and Opacity
       sections [{
         type=MovieSceneColorSection
         end_frame=60
         start_frame=0
         channels [
          Color.R {type=MovieSceneScriptingFloatChannel}
          Color.G {type=MovieSceneScriptingFloatChannel}
          Color.B {type=MovieSceneScriptingFloatChannel}
          Color.A {type=MovieSceneScriptingFloatChannel, keys [[1.936,0.25],[60,0.6]]}
         ]
        }]
      }
     ]
    }
    CanvasPanelSlot_3 {
     id=#22
     parent_id=#23
     possessed_class=CanvasPanelSlot
     tracks [{
       type=MovieSceneMarginTrack
       display_name="Offsets (Layout Data)"
       sections [{
         type=MovieSceneMarginSection
         end_frame=60
         start_frame=0
         channels [
          Left {type=MovieSceneScriptingFloatChannel, keys [[0,-78],[29,-90],[30,136],[60,120]]}
          Top {type=MovieSceneScriptingFloatChannel, default=0}
          Right {type=MovieSceneScriptingFloatChannel, default=0}
          Bottom {type=MovieSceneScriptingFloatChannel, default=0}
         ]
        }]
      }]
    }
    PartIconPrevious2_Legacy {
     id=#23
     possessed_class=BP_JamPartIcon_C
     tracks [
      {
       type=MovieSceneColorTrack
       display_name=Color and Opacity
       sections [{
         type=MovieSceneColorSection
         end_frame=60
         start_frame=0
         channels [
          Color.R {type=MovieSceneScriptingFloatChannel}
          Color.G {type=MovieSceneScriptingFloatChannel}
          Color.B {type=MovieSceneScriptingFloatChannel}
          Color.A {type=MovieSceneScriptingFloatChannel, keys [[0,0.6],[29,0],[30,0],[60,0.25]]}
         ]
        }]
      }
      {
       type=MovieScene2DTransformTrack
       sections [{
         type=MovieScene2DTransformSection
         end_frame=29
         start_frame=0
         channels [
          Translation.X {type=MovieSceneScriptingFloatChannel, const [0,1]}
          Translation.Y {type=MovieSceneScriptingFloatChannel, const [0,1]}
          Angle {type=MovieSceneScriptingFloatChannel, const [0,1]}
          Scale.X {type=MovieSceneScriptingFloatChannel, keys [[0,1.33],[29,1]]}
          Scale.Y {type=MovieSceneScriptingFloatChannel, keys [[0,1.33],[29,1]]}
          Shear.X {type=MovieSceneScriptingFloatChannel, const [0,1]}
          Shear.Y {type=MovieSceneScriptingFloatChannel, const [0,1]}
         ]
        }]
      }
     ]
    }
    PartIconNext_Legacy {
     id=#19
     possessed_class=BP_JamPartIcon_C
     tracks [
      {
       type=MovieScene2DTransformTrack
       sections [{
         type=MovieScene2DTransformSection
         end_frame=60
         start_frame=0
         channels [
          Angle {type=MovieSceneScriptingFloatChannel}
          Scale.X {type=MovieSceneScriptingFloatChannel, keys [[0,2.3],[60,1]]}
          Scale.Y {type=MovieSceneScriptingFloatChannel, keys [[0,2.3],[60,1]]}
          Shear.X {type=MovieSceneScriptingFloatChannel}
          Shear.Y {type=MovieSceneScriptingFloatChannel}
         ]
        }]
      }
      {
       type=MovieSceneColorTrack
       display_name=Color and Opacity
       sections [{
         type=MovieSceneColorSection
         end_frame=60
         start_frame=0
         channels [
          Color.R {type=MovieSceneScriptingFloatChannel}
          Color.G {type=MovieSceneScriptingFloatChannel}
          Color.B {type=MovieSceneScriptingFloatChannel}
          Color.A {type=MovieSceneScriptingFloatChannel, keys [[0,1],[60,0.6]]}
         ]
        }]
      }
     ]
    }
   ]
   display_rate {numerator=240}
  }
 }
 A_Cycle_Forward_Legacy {
  binding_count=8
  length_seconds=0.25
  object=&self:A_Cycle_Forward_Legacy
  sequence {
   bindings [
    PartIconMain_Legacy {
     id=#16
     possessed_class=BP_JamPartIcon_C
     tracks [
      {
       type=MovieScene2DTransformTrack
       sections [{
         type=MovieScene2DTransformSection
         end_frame=60
         start_frame=0
         channels [
          Translation.X {type=MovieSceneScriptingFloatChannel}
          Translation.Y {type=MovieSceneScriptingFloatChannel}
          Angle {type=MovieSceneScriptingFloatChannel}
          Scale.X {type=MovieSceneScriptingFloatChannel, keys [[0,1],[60,2.65]]}
          Scale.Y {type=MovieSceneScriptingFloatChannel, keys [[0,1],[60,2.65]]}
          Shear.X {type=MovieSceneScriptingFloatChannel}
          Shear.Y {type=MovieSceneScriptingFloatChannel}
         ]
        }]
      }
      {
       type=MovieSceneColorTrack
       display_name=Color and Opacity
       sections [{
         type=MovieSceneColorSection
         end_frame=60
         start_frame=0
         channels [
          Color.R {type=MovieSceneScriptingFloatChannel}
          Color.G {type=MovieSceneScriptingFloatChannel}
          Color.B {type=MovieSceneScriptingFloatChannel}
          Color.A {type=MovieSceneScriptingFloatChannel, keys [[0,0.6],[60,1]]}
         ]
        }]
      }
     ]
    }
    CanvasPanelSlot_1 {
     id=#17
     parent_id=#16
     possessed_class=CanvasPanelSlot
     tracks [{
       type=MovieSceneMarginTrack
       display_name="Offsets (Layout Data)"
       sections [{
         type=MovieSceneMarginSection
         end_frame=60
         start_frame=0
         channels [
          Left {type=MovieSceneScriptingFloatChannel, keys [[0,-90],[60,0]]}
          Top {type=MovieSceneScriptingFloatChannel, default=0}
          Right {type=MovieSceneScriptingFloatChannel, default=0}
          Bottom {type=MovieSceneScriptingFloatChannel, default=0}
         ]
        }]
      }]
    }
    CanvasPanelSlot_2 {
     id=#18
     parent_id=#19
     possessed_class=CanvasPanelSlot
     tracks [{
       type=MovieSceneMarginTrack
       display_name="Offsets (Layout Data)"
       sections [{
         type=MovieSceneMarginSection
         end_frame=60
         start_frame=0
         channels [
          Left {type=MovieSceneScriptingFloatChannel, keys [[0,-144],[60,-90]]}
          Top {type=MovieSceneScriptingFloatChannel, default=0}
          Right {type=MovieSceneScriptingFloatChannel, default=0}
          Bottom {type=MovieSceneScriptingFloatChannel, default=0}
         ]
        }]
      }]
    }
    CanvasPanelSlot_10 {
     id=#20
     parent_id=#21
     possessed_class=CanvasPanelSlot
     tracks [{
       type=MovieSceneMarginTrack
       display_name="Offsets (Layout Data)"
       sections [{
         type=MovieSceneMarginSection
         end_frame=60
         start_frame=0
         channels [
          Left {type=MovieSceneScriptingFloatChannel, keys [[0,0],[60,90]]}
          Top {type=MovieSceneScriptingFloatChannel, default=0}
          Right {type=MovieSceneScriptingFloatChannel, default=0}
          Bottom {type=MovieSceneScriptingFloatChannel, default=0}
         ]
        }]
      }]
    }
    PartIconPrevious_Legacy {
     id=#21
     possessed_class=BP_JamPartIcon_C
     tracks [
      {
       type=MovieScene2DTransformTrack
       sections [{
         type=MovieScene2DTransformSection
         end_frame=60
         start_frame=0
         channels [
          Translation.X {type=MovieSceneScriptingFloatChannel}
          Translation.Y {type=MovieSceneScriptingFloatChannel}
          Angle {type=MovieSceneScriptingFloatChannel}
          Scale.X {type=MovieSceneScriptingFloatChannel, keys [[0,2.65],[60,1]]}
          Scale.Y {type=MovieSceneScriptingFloatChannel, keys [[0,2.65],[60,1]]}
          Shear.X {type=MovieSceneScriptingFloatChannel}
          Shear.Y {type=MovieSceneScriptingFloatChannel}
         ]
        }]
      }
      {
       type=MovieSceneColorTrack
       display_name=Color and Opacity
       sections [{
         type=MovieSceneColorSection
         end_frame=60
         start_frame=0
         channels [
          Color.R {type=MovieSceneScriptingFloatChannel}
          Color.G {type=MovieSceneScriptingFloatChannel}
          Color.B {type=MovieSceneScriptingFloatChannel}
          Color.A {type=MovieSceneScriptingFloatChannel, keys [[0,1],[60,0.6]]}
         ]
        }]
      }
     ]
    }
    CanvasPanelSlot_3 {
     id=#22
     parent_id=#23
     possessed_class=CanvasPanelSlot
     tracks [{
       type=MovieSceneMarginTrack
       display_name="Offsets (Layout Data)"
       sections [{
         type=MovieSceneMarginSection
         end_frame=60
         start_frame=0
         channels [
          Left {type=MovieSceneScriptingFloatChannel, keys [[0,90],[29,115],[30,-170],[60,-144]]}
          Top {type=MovieSceneScriptingFloatChannel, default=0}
          Right {type=MovieSceneScriptingFloatChannel, default=0}
          Bottom {type=MovieSceneScriptingFloatChannel, default=0}
         ]
        }]
      }]
    }
    PartIconPrevious2_Legacy {
     id=#23
     possessed_class=BP_JamPartIcon_C
     tracks [{
       type=MovieSceneColorTrack
       display_name=Color and Opacity
       sections [{
         type=MovieSceneColorSection
         end_frame=60
         start_frame=0
         channels [
          Color.R {type=MovieSceneScriptingFloatChannel}
          Color.G {type=MovieSceneScriptingFloatChannel}
          Color.B {type=MovieSceneScriptingFloatChannel}
          Color.A {type=MovieSceneScriptingFloatChannel, keys [[0,0.6],[29,0],[30,0],[60,0.6]]}
         ]
        }]
      }]
    }
    PartIconNext_Legacy {
     id=#19
     possessed_class=BP_JamPartIcon_C
     tracks [{
       type=MovieSceneColorTrack
       display_name=Color and Opacity
       sections [{
         type=MovieSceneColorSection
         end_frame=60
         start_frame=0
         channels [
          Color.R {type=MovieSceneScriptingFloatChannel}
          Color.G {type=MovieSceneScriptingFloatChannel}
          Color.B {type=MovieSceneScriptingFloatChannel}
          Color.A {type=MovieSceneScriptingFloatChannel, const [0.6,2]}
         ]
        }]
      }]
    }
   ]
   display_rate {numerator=240}
  }
 }
]
graphs [
 DecrementMultiEmoteIndex {nodes [
   DecrementMultiEmoteIndex() {is=title K2Node_FunctionEntry}
   n1 = GetMultiEmoteIndex!() -> (MultiEmoteIndex:int=0)
   > SetMultiEmoteIndex!(
    execute
    self:Jam Part Switcher VM Object Reference=Get JamPartSwitcherVM() -> (JamPartSwitcherVM:Jam Part Switcher VM Object Reference)
    InMultiEmoteIndex:int=GetLoopedIndex!(RawIndex:int=int - int(A=@n1)) -> (LoopedIndex:int=0)
   ) -> (then)
  ]}
 EventGraph {nodes [
   n1 = "On Button Clicked (InputReflectorButton)"!() -> (OutputDelegate, then, Button:Common Button Legacy Object Reference) {is=title K2Node_ComponentBoundEvent}
   IncrementMultiEmoteIndex(execute=[@n7.then, @n1.then, @n3.then, @n4.then, @n5.then, @n6.then])
   > UpdateVisuals()
   > n2 = "Play Animation with Finished event (legacy)"!(
    execute:Exec
    Widget:User Widget Object Reference=Self-Reference()
    InAnimation:Widget Animation Object Reference=Select(
     Option 0:Widget Animation Object Reference=Get A_Cycle_Back_Legacy() -> (A_Cycle_Back_Legacy:Widget Animation Object Reference)
     Option 1:Widget Animation Object Reference=Get A_Cycle_Back_All() -> (A_Cycle_Back_All:Widget Animation Object Reference)
     Index=IsPartEnabled!(LoopType:EFMJamLoopType Enum=Num) -> (IsEnabled:bool)
    ) -> (ReturnValue:Widget Animation Object Reference)
    StartAtTime:float=0
    NumLoopsToPlay:int=1
    PlayMode:EUMGSequencePlayMode Enum
    PlaybackSpeed:float=1
   ) -> (then:Exec, Result:UMGSequence Player Object Reference, Finished:Exec) {is=title K2Node_PlayAnimation}
   > Set CycleAnim_Handle(CycleAnim_Handle:UMGSequence Player Object Reference=@n2) -> (Output_Get:UMGSequence Player Object Reference)
   > TriggerFlowriderInteraction!(
    execute
    self
    Element:Widget Object Reference=Get Switcher_Button() -> (Switcher_Button:Button Object Reference)
    Value:String=Get Source String from Text!(Text:Text=Get TextToUse() -> (TextToUse:Text)) -> (ReturnValue:String) {is=title K2Node_CallFunction}
   ) -> (then)
   n3 = "On Button Touched (PartIconNext2)"() {is=title K2Node_ComponentBoundEvent}
   n4 = "On Button Touched (PartIconNext1)"() {is=title K2Node_ComponentBoundEvent}
   n5 = "On Button Touched (PartIconPrevious1)"() {is=title K2Node_ComponentBoundEvent}
   n6 = "On Button Touched (PartIconMain)"() {is=title K2Node_ComponentBoundEvent}
   "On Clicked (Switcher_Button)"() {is=title K2Node_ComponentBoundEvent}
   > n7 = PlayAnimationForward(InAnimation=Get A_ButtonClick_Pulse() -> (A_ButtonClick_Pulse:Widget Animation Object Reference))
   "On Hovered (Switcher_Button)"() {is=title K2Node_ComponentBoundEvent}
   > n8 = Branch(Condition=IsUsingTouch(PlayerController=GetOwningPlayer()))
   PlayAnimationForward(execute=@n8.else, InAnimation=Get A_Button_Hover() -> (A_Button_Hover:Widget Animation Object Reference))
   "On Unhovered (Switcher_Button)"() {is=title K2Node_ComponentBoundEvent}
   > n9 = Branch(Condition=IsUsingTouch(PlayerController=GetOwningPlayer()))
   PlayAnimationReverse(execute=@n9.else, InAnimation=Get A_Button_Hover())
   n10 = On Input Type Changed!() -> (OutputDelegate, then, bNewInputType:ECommonInputType Enum) {is=title K2Node_CustomEvent}
   n11 = Get PC Console Background Color() -> (PC Console Background Color:Linear Color Structure)
   > SetBackgroundColor!(
    execute
    self:Button Object Reference=Get Switcher_Button()
    InBackgroundColor:Linear Color Structure=Select!(
     MouseAndKeyboard:Linear Color Structure=@n11
     Gamepad:Linear Color Structure=@n11
     Touch:Linear Color Structure=(R=0,G=0,B=0,A=0)
     Count:Linear Color Structure=@n11
     Index:ECommonInputType Enum=@n10.bNewInputType
    ) -> (ReturnValue:Linear Color Structure)
   ) -> (then)
   Event Tick()
   > n12 = Branch(Condition=IsMobileGame())
   n13 = GetMousePositionOnViewport!(execute=@n12.else) -> (then, ReturnValue:Vector 2D Structure)
   n14 = Get Switcher_Button()
   n15 = GetPaintSpaceGeometry!(self:Widget Object Reference=@n14) -> (ReturnValue:"Geometry Structure (by ref)")
   n16 = vector2d * vector2d(
    A:Vector 2D Structure=GetLocalSize!(Geometry:"Geometry Structure (by ref)"=@n15) -> (ReturnValue:Vector 2D Structure)
    B:Vector 2D Structure=(X=0.5,Y=0.5)
   ) -> (ReturnValue:Vector 2D Structure) {is=title K2Node_PromotableOperator}
   n17 = vector2d + vector2d(
    A:Vector 2D Structure=@n16
    B:Vector 2D Structure=LocalToViewport!(
     Geometry:"Geometry Structure (by ref)"=GetTickSpaceGeometry!(self:Widget Object Reference=@n14) -> (ReturnValue:"Geometry Structure (by ref)")
     LocalCoordinate:Vector 2D Structure=GetLocalTopLeft!(Geometry:"Geometry Structure (by ref)"=@n15) -> (ReturnValue:Vector 2D Structure)
    ) -> (PixelPosition:Vector 2D Structure, ViewportPosition:Vector 2D Structure).ViewportPosition
   ) -> (ReturnValue:Vector 2D Structure) {is=title K2Node_PromotableOperator}
   > n18 = Branch(Condition="float >= float"(A=BreakVector2D(InVec=@n16), B=Distance2D!(V1:Vector 2D Structure=@n17, V2:Vector 2D Structure=@n13) -> (ReturnValue:double)))
   > Branch(Condition=NOT Boolean(A=Get IsMouseOver() -> (IsMouseOver:bool)))
   > Set IsMouseOver(IsMouseOver:bool=true) -> (Output_Get:bool)
   > Set Visibility(Visibility=Visible, self=Get Switcher_Button())
   Branch(execute=@n18.else, Condition=Get IsMouseOver())
   > Set IsMouseOver(IsMouseOver=false)
   > Set Visibility(Visibility=SelfHitTestInvisible, self=Get Switcher_Button())
   Event On Activated()
   > SetupInitialIndex()
   > Set CachedLoopOptionsComponent(CachedLoopOptionsComponent:Sparks Controller Component Sparks Songs Emote Options Object Reference=GetComponentByClass!(
     self:Actor Object Reference=GetOwningPlayer()
     ComponentClass:Actor Component Class Reference=/Script/SparksSongsEmoteWheelRuntime.SparksControllerComponent_SparksSongsEmoteOptions
    ) -> (ReturnValue:Sparks Controller Component Sparks Songs Emote Options Object Reference)) -> (Output_Get:Sparks Controller Component Sparks Songs Emote Options Object Reference)
   > ResetVisuals()
   Event Construct()
   > Set PC Console Background Color(PC Console Background Color:Linear Color Structure=Get BackgroundColor!(self:Button Object Reference=Get Switcher_Button()) -> (BackgroundColor:Linear Color Structure)) -> (Output_Get:Linear Color Structure)
   > Bind Event to On Input Method Changed(self=Get CommonInputSubsystem(), Delegate=@n10)
   > On Input Type Changed!(execute, self, bNewInputType:ECommonInputType Enum=GetCurrentInputType(self=Get CommonInputSubsystem())) -> (then) {is=title K2Node_CallFunction}
   > SetActiveWidgetIndex(
    self:Common Visibility Switcher Object Reference=Get PartSetVisibilitySwitcher() -> (PartSetVisibilitySwitcher:Common Visibility Switcher Object Reference)
    Index=SelectInt(A=1, B=0, bPickA=IsPartEnabled(LoopType=Num))
   )
   Event PreConstruct()
   > SetMobileStyle() {is=title K2Node_CallFunction}
  ]}
 GetIndexForLoopType {nodes [
   n1 = GetIndexForLoopType!() -> (then, LoopType:EFMJamLoopType Enum=Lead) {is=title K2Node_FunctionEntry}
   > n2 = Sequence()
   > n3 = Get!(execute:Exec) -> (then:Exec, else:Exec, CachedLoopOptionsComponent:Sparks Controller Component Sparks Songs Emote Options Object Reference)
   > Return Node!(
    execute
    Index:int=GetMultiEmoteIndexForLoopType!(
     self:Sparks Controller Component Sparks Songs Emote Options Object Reference=@n3
     LoopType:EFMJamLoopType Enum=Get LoopType() -> (LoopType:EFMJamLoopType Enum=Lead)
    ) -> (ReturnValue:int=0)
   )
   Return Node(
    execute=@n2.then_1
    Index=Find Item!(
     TargetArray:Array of EFMJamLoopType Enums=Get LoopTypeList() -> (LoopTypeList:Array of EFMJamLoopType Enums=Lead)
     ItemToFind:"EFMJamLoopType Enum (by ref)"=@n1
    ) -> (ReturnValue)
   )
  ]}
 GetLoopTypeForIndex {nodes [
   n1 = GetLoopTypeForIndex!() -> (then, Index:int) {is=title K2Node_FunctionEntry}
   > n2 = Sequence()
   > n3 = Get()
   > n4 = GetLoopTypeForMultiEmoteIndex!(execute, self:Sparks Controller Component Sparks Songs Emote Options Object Reference=@n3, Index:int=Get Index() -> (Index:int=0)) -> (then, OutLoopType:EFMJamLoopType Enum=Lead, ReturnValue:bool)
   > Branch(Condition=@n4.ReturnValue)
   > Return Node!(execute, LoopType:EFMJamLoopType Enum=@n4)
   n5 = Get LoopTypeList() -> (LoopTypeList=Lead)
   n6 = Branch(execute=@n2.then_1, Condition=Is Valid Index!(TargetArray:Array of EFMJamLoopType Enums=@n5, IndexToTest:int=@n1) -> (ReturnValue:bool))
   > Return Node(LoopType="Get (a copy)"(Array:Array of EFMJamLoopType Enums=@n5, Dimension 1=@n1) -> (Output:EFMJamLoopType Enum))
   Return Node(execute=@n6.else, LoopType=Num)
  ]}
 GetLoopedIndex {nodes [
   n1 = GetLoopedIndex!() -> (then, RawIndex:int) {is=title K2Node_FunctionEntry}
   > Set TempLoopedIndex(TempLoopedIndex:int=@n1) -> (Output_Get:int=0)
   > Set MinIndex(MinIndex:int=SelectInt(A=0, B=1, bPickA=IsPartEnabled(LoopType=Num))) -> (Output_Get:int=0)
   n2 = Get MaxIndex() -> (MaxIndex:int=0)
   n3 = Get MinIndex() -> (MinIndex:int=0)
   > Set NumParts(NumParts:int=int + int(A=int - int(A=@n2, B=@n3))) -> (Output_Get:int=0)
   > n4 = While Loop!(execute:Exec, Condition:bool="integer < integer"(A=Get TempLoopedIndex() -> (TempLoopedIndex:int=0), B=Get MinIndex() -> (MinIndex=0))) -> (LoopBody:Exec, Completed:Exec) {is=title K2Node_MacroInstance}
   n5 = While Loop(execute=@n4.Completed, Condition="integer > integer"(A=Get TempLoopedIndex() -> (TempLoopedIndex=0), B=Get MaxIndex() -> (MaxIndex=0)))
   Return Node!(execute=@n5.Completed, LoopedIndex:int=Get TempLoopedIndex() -> (TempLoopedIndex=0))
   > Set TempLoopedIndex(TempLoopedIndex=int - int(A=Get TempLoopedIndex() -> (TempLoopedIndex=0), B=Get NumParts() -> (NumParts:int=0))) -> (Output_Get=0)
   Set TempLoopedIndex(execute=@n4.LoopBody, TempLoopedIndex=int + int(A=Get TempLoopedIndex() -> (TempLoopedIndex=0), B=Get NumParts() -> (NumParts=0))) -> (Output_Get=0)
  ]}
 GetMultiEmoteIndex {nodes [
   GetMultiEmoteIndex() {is=title K2Node_FunctionEntry}
   > Return Node!(
    execute
    MultiEmoteIndex:int=GetMultiEmoteIndex!(self:Jam Part Switcher VM Object Reference=Get JamPartSwitcherVM()) -> (ReturnValue:int=0) {is=title K2Node_CallFunction}
   )
   Return Node(MultiEmoteIndex=0)
  ]}
 IncrementMultiEmoteIndex {nodes [
   IncrementMultiEmoteIndex() {is=title K2Node_FunctionEntry}
   n1 = GetMultiEmoteIndex!() -> (MultiEmoteIndex:int=0)
   > SetMultiEmoteIndex(self=Get JamPartSwitcherVM(), InMultiEmoteIndex=GetLoopedIndex(RawIndex=int + int(A=@n1)) -> (LoopedIndex=0) {is=title K2Node_CallFunction})
  ]}
 IsPartEnabled {nodes [
   n1 = IsPartEnabled!() -> (then, LoopType:EFMJamLoopType Enum=Lead) {is=title K2Node_FunctionEntry}
   > n2 = Cast To FortPlayerController(Object=GetOwningPlayer())
   > n3 = IsLoopTypeUsableByEmote!(execute, LoopType:EFMJamLoopType Enum=@n1, PlayerController:Fort Player Controller Object Reference=@n2) -> (then, ReturnValue:bool)
   > Return Node!(execute, IsEnabled:bool=@n3)
   FortShippingLog!(execute=@n2.CastFailed, WorldContextObject:Object Reference, InString:String=Could not get FortPlayerController, bLogAsWarning:bool=true) -> (then)
   > Return Node(IsEnabled=true)
  ]}
 ResetVisuals {nodes [
   ResetVisuals() {is=title K2Node_FunctionEntry}
   > UpdateVisuals()
  ]}
 SetLayout {nodes [SetLayout!() -> (then, ShowContainer:bool=~) {is=title K2Node_FunctionEntry}]}
 SetMobileStyle {nodes [
   SetMobileStyle() {is=title K2Node_FunctionEntry}
   > Branch(Condition=IsUsingCompactUI())
   n1 = Get SizeBox_Switcher() -> (SizeBox_Switcher:Size Box Object Reference)
   > SetWidthOverride(self=@n1, InWidthOverride=105)
   > SetHeightOverride(self=@n1, InHeightOverride=105)
   > SetWidthOverride(self=Get SizeBox_InstrumentText() -> (SizeBox_InstrumentText:Size Box Object Reference), InWidthOverride=200)
  ]}
 SetSelectedCategory {nodes [SetSelectedCategory!() -> (then, Category:int) {is=title K2Node_FunctionEntry}]}
 SetupInitialIndex {nodes [
   SetupInitialIndex() {is=title K2Node_FunctionEntry}
   > Set StartingIndex(StartingIndex:int=GetMultiEmoteIndex() -> (MultiEmoteIndex=0)) -> (Output_Get:int=0)
   n1 = GetLoopTypeForIndex!(Index:int=GetMultiEmoteIndex() -> (MultiEmoteIndex=0)) -> (LoopType:EFMJamLoopType Enum=Lead) {is=title K2Node_CallFunction}
   > Branch(Condition=NOT Boolean(A=IsPartEnabled(LoopType=@n1) {is=title K2Node_CallFunction}))
   > IncrementMultiEmoteIndex() {is=title K2Node_CallFunction}
   n2 = GetMultiEmoteIndex() -> (MultiEmoteIndex=0)
   n3 = IsPartEnabled(LoopType=GetLoopTypeForIndex(Index=@n2) -> (LoopType=Lead))
   n4 = Get StartingIndex() -> (StartingIndex:int=0)
   > While Loop(Condition=AND Boolean(A="Not Equal (Integer)"(A=@n2, B=@n4), B=NOT Boolean(A=@n3)))
   > IncrementMultiEmoteIndex()
  ]}
 TriggerFlowriderInteraction {nodes [
   n1 = TriggerFlowriderInteraction!() -> (then, Element:Widget Object Reference, Value:String) {is=title K2Node_FunctionEntry}
   n2 = GetAnalyticsEWActivationCount!(self:Fort Picker Context Object Reference=Get FortPickerContext() -> (ReturnValue:Fort Picker Context Object Reference) {is=title K2Node_GetSubsystem}) -> (ReturnValue:int=0)
   n3 = Make Flow Analytics Additional Info!() -> (FlowAnalyticsAdditionalInfo:Flow Analytics Additional Info Structure) {is=title K2Node_MakeStruct}
   > Fire Event Element Interacted(
    Element=@n1
    AdditionalInfo="Add String (Non-Localized)"!(
     Info:"Flow Analytics Additional Info Structure (by ref)"="Add String (Non-Localized)"(Info=@n3, Key=Button Name, Value=@n1.Value)
     Key:String=emote_wheel_session_id
     Value:String=GetAnalyticsSessionID!(ActivationCount:int=@n2) -> (ReturnValue:String)
    ) -> (ReturnValue:Flow Analytics Additional Info Structure)
   )
  ]}
 UpdateIcon {nodes [
   UpdateIcon!() -> (then, IconIndex:int) {is=title K2Node_FunctionEntry}
   > n1 = Sequence!(execute) -> (then_0, then_1, then_2:Exec)
   n2 = Get PartIconNext2() -> (PartIconNext2:BP Jam Part Icon Object Reference)
   n3 = Get PartIconNext1() -> (PartIconNext1:BP Jam Part Icon Object Reference)
   n4 = Get PartIconPrevious1() -> (PartIconPrevious1:BP Jam Part Icon Object Reference)
   n5 = Get IconIndex() -> (IconIndex:int=0)
   n6 = Get PartIconPrevious2() -> (PartIconPrevious2:BP Jam Part Icon Object Reference)
   n7 = Get PartIconMain() -> (PartIconMain:BP Jam Part Icon Object Reference)
   n8 = Get IconIndex() -> (IconIndex=0)
   n9 = Get PartIconPrevious_Legacy() -> (PartIconPrevious_Legacy:BP Jam Part Icon Object Reference)
   n10 = Get PartIconPrevious2_Legacy() -> (PartIconPrevious2_Legacy:BP Jam Part Icon Object Reference)
   n11 = Get PartIconNext_Legacy() -> (PartIconNext_Legacy:BP Jam Part Icon Object Reference)
   n12 = Get PartIconMain_Legacy() -> (PartIconMain_Legacy:BP Jam Part Icon Object Reference)
   > Set Icon(Icon:BP Jam Part Icon Object Reference=Select!(
     Option 0:BP Jam Part Icon Object Reference=Select!(Option 0=@n12, Option 1=@n9, Option 2:BP Jam Part Icon Object Reference=@n10, Option 3:BP Jam Part Icon Object Reference=@n11, Index:int=@n8) -> (ReturnValue)
     Option 1:BP Jam Part Icon Object Reference=Select!(Option 0=@n7, Option 1=@n4, Option 2=@n6, Option 3=@n3, Option 4:BP Jam Part Icon Object Reference=@n2, Index=@n5) -> (ReturnValue)
     Index:bool=IsPartEnabled(LoopType=Num)
    ) -> (ReturnValue:BP Jam Part Icon Object Reference)) -> (Output_Get:BP Jam Part Icon Object Reference)
   n13 = int + int(A=GetMultiEmoteIndex() -> (MultiEmoteIndex=0), B=Get IconIndex() -> (IconIndex=0))
   > Set LoopType(LoopType:EFMJamLoopType Enum=GetLoopTypeForIndex(Index=GetLoopedIndex(RawIndex=@n13) -> (LoopedIndex=0)) -> (LoopType=Lead)) -> (Output_Get:EFMJamLoopType Enum=Lead)
   > Set Active Icon by LoopType!(
    execute
    self:BP Jam Part Icon Object Reference=Get Icon() -> (Icon:BP Jam Part Icon Object Reference)
    LoopType:EFMJamLoopType Enum=Get LoopType() -> (LoopType=Lead)
   ) -> (then) {is=title K2Node_CallFunction}
   n14 = Branch(execute=@n1.then_1, Condition="Equal (Integer)"(A=Get IconIndex() -> (IconIndex=0)))
   n15 = IsPartEnabled(LoopType=Get LoopType() -> (LoopType=Lead))
   > n16 = Branch(Condition=@n15)
   Set ColorToUse(execute=@n16.else, ColorToUse:Linear Color Structure=(R=0,G=0,B=0,A=0.75)) -> (Output_Get:Linear Color Structure)
   n17 = Branch(execute=@n14.else, Condition=@n15)
   Set ColorToUse(execute=@n17.else, ColorToUse=(R=0,G=0,B=0,A=1))
   SetFocusIconColor!(
    execute=@n1.then_2
    self:BP Jam Part Icon Object Reference=Get Icon()
    InColorAndOpacity:Linear Color Structure=Get ColorToUse() -> (ColorToUse:Linear Color Structure)
   ) -> (then)
  ]}
 UpdateText {nodes [
   UpdateText() {is=title K2Node_FunctionEntry}
   > n1 = Sequence()
   > Set LoopType(LoopType=GetLoopTypeForIndex(Index=GetMultiEmoteIndex() -> (MultiEmoteIndex=0)) -> (LoopType=Lead)) -> (Output_Get=Lead)
   n2 = Get Instrument_Text() -> (Instrument_Text:Common Text Object Reference)
   Set Text(execute=@n1.then_2, Text=Get TextToUse(), self=@n2)
   > SetColorAndOpacity(self=@n2, InColorAndOpacity=Get ColorToUse() -> (ColorToUse:Slate Color Structure))
   n3 = Cast To FortPlayerController(execute=@n1.then_1, Object=GetOwningPlayer())
   > n4 = IsJammingBlocked!(execute, PlayerController:Fort Player Controller Object Reference=@n3) -> (True:Exec, False:Exec)
   n5 = Branch(execute=@n4.False, Condition=IsPartEnabled(LoopType=Get LoopType() -> (LoopType=Lead)))
   n6 = Get LoopType() -> (LoopType=Lead)
   > Set TextToUse(TextToUse:Text="Get (a ref)"(
     Array:Array of Texts=Get PartTextList() -> (PartTextList:Array of Texts)
     Dimension 1="To Integer (Byte)"!(InByte:Byte=@n6) -> (ReturnValue:int=0)
    ) -> (Output:"Text (by ref)")) -> (Output_Get:Text)
   Set TextToUse(execute=[@n5.else, @n4.True, @n3.CastFailed], TextToUse=Get DisabledText() -> (DisabledText:Text))
   > Set ColorToUse(ColorToUse:Slate Color Structure=Get DisabledColor() -> (DisabledColor:Slate Color Structure)) -> (Output_Get:Slate Color Structure)
  ]}
 UpdateVisuals {nodes [
   UpdateVisuals() {is=title K2Node_FunctionEntry}
   > n1 = Sequence!(execute) -> (then_0, then_1, then_2, then_3:Exec, then_4:Exec, then_5:Exec)
   > UpdateText() {is=title K2Node_CallFunction}
   UpdateIcon!(execute=@n1.then_1, self, IconIndex:int=0) -> (then) {is=title K2Node_CallFunction}
   UpdateIcon(execute=@n1.then_2, IconIndex=1)
   UpdateIcon(execute=@n1.then_3, IconIndex=2)
   UpdateIcon(execute=@n1.then_4, IconIndex=3)
   Branch(execute=@n1.then_5, Condition=IsPartEnabled(LoopType=Num))
   > UpdateIcon(IconIndex=4)
  ]}
]
variables [
 CachedLoopOptionsComponent {type=SparksControllerComponent_SparksSongsEmoteOptions, default=null}
 CycleAnim_Handle {type=UMGSequencePlayer, default=null}
 DisabledText {type=string, default=(Disabled)}
 EmoteWheelSelectionVM {type=EmoteWheelItemSelectionVM, default=null}
 IsMouseOver {type=boolean}
 LoopTypeList {type="array<EFMJamLoopType>", default [Num,Beat,Bass,Misc,Lead]}
 PartTextList {type="array<string>", default [Vocals,Lead,Bass,Drums,All]}
 PC Console Background Color {type=LinearColor, default=<rgba 0 0 0 0>}
 TextToUse {type=string, default=""}
]
widgets {
 info {inheritedWidgetCount=0, parentClass=&CommonActivatableWidget, rootWidgetClass=&SizeBox, widgetCount=20}
 widgets [
  SizeBox_Main {
   widgetClassPath=&SizeBox
   bIsVariable=true
   widget=SizeBox_Main
   properties {bOverride_HeightOverride=true, bOverride_WidthOverride=true, heightOverride=352, widthOverride=352}
  }
  Main_Overlay {widgetClassPath=&Overlay, parent=SizeBox_Main, widget=Main_Overlay, properties {clipping=OnDemand}, slot {class=SizeBoxSlot}}
  SelectedInstrument_Background {
   widgetClassPath=&Image
   bIsVariable=true
   parent=Main_Overlay
   widget=SelectedInstrument_Background
   properties {
    visibility=SelfHitTestInvisible
    brush {
     bIsDynamicallyLoaded=false
     drawAs=Image
     imageType=FullColor
     mirroring=NoMirror
     resourceName=None
     resourceObject=/FMJamSystem/Widgets/Materials/MI_UI_PartSwitcherShape
     tiling=NoTile
     imageSize=<312 312>
     margin=<ltrb 0 0 0 0>
     outlineSettings {
      bUseBrushTransparency=false
      roundingType=HalfHeightRadius
      width=0
      color {colorUseRule=UseColor_Specified, specifiedColor=<rgba 0 0 0 0>}
      cornerRadii=<0 0 0 0>
     }
     tintColor {colorUseRule=UseColor_Specified, specifiedColor=<rgba 1 1 1 1>}
     uVRegion {bIsValid=false, max=<0 0>, min=<0 0>}
    }
    colorAndOpacity=<rgba 0.01 0.01 0.01 0.85>
   }
   slot {class=OverlaySlot, properties {horizontalAlignment=HAlign_Center, verticalAlignment=VAlign_Center, padding=<ltrb 0 -52 0 -52>}}
  }
  PartSetVisibilitySwitcher {
   widgetClassPath=&CommonVisibilitySwitcher
   bIsVariable=true
   parent=Main_Overlay
   widget=PartSetVisibilitySwitcher
   slot {class=OverlaySlot, properties {horizontalAlignment=HAlign_Fill, verticalAlignment=VAlign_Center}}
  }
  CanvasPanel-LegacyParts {
   widgetClassPath=&CanvasPanel
   parent=PartSetVisibilitySwitcher
   widget=CanvasPanel-LegacyParts
   slot {class=CommonVisibilitySwitcherSlot}
  }
  PartIconNext_Legacy {
   widgetClassPath=&/BP_JamPartIcon
   bIsVariable=true
   parent=CanvasPanel-LegacyParts
   widget=PartIconNext_Legacy
   properties {colorAndOpacity=<rgba 1 1 1 0.6>, imageSize=<48 48>}
   slot {
    class=CanvasPanelSlot
    properties {bAutoSize=true, layoutData {alignment=<0.5 0.5>, anchors {maximum=<0.5 0.5>, minimum=<0.5 0.5>}, offsets=<ltrb -78 0 36 36>}}
   }
  }
  PartIconPrevious2_Legacy {
   widgetClassPath=&BP_JamPartIcon
   bIsVariable=true
   parent=CanvasPanel-LegacyParts
   widget=PartIconPrevious2_Legacy
   properties {colorAndOpacity=<rgba 1 1 1 0.25>, imageSize=<36 36>}
   slot {
    class=CanvasPanelSlot
    properties {bAutoSize=true, layoutData {alignment=<0.5 0.5>, anchors {maximum=<0.5 0.5>, minimum=<0.5 0.5>}, offsets=<ltrb 120 0 36 36>}}
   }
  }
  PartIconPrevious_Legacy {
   widgetClassPath=&BP_JamPartIcon
   bIsVariable=true
   parent=CanvasPanel-LegacyParts
   widget=PartIconPrevious_Legacy
   properties {colorAndOpacity=<rgba 1 1 1 0.6>, imageSize=<48 48>}
   slot {
    class=CanvasPanelSlot
    properties {bAutoSize=true, layoutData {alignment=<0.5 0.5>, anchors {maximum=<0.5 0.5>, minimum=<0.5 0.5>}, offsets=<ltrb 78 0 96 96>}}
   }
  }
  PartIconMain_Legacy {
   widgetClassPath=&BP_JamPartIcon
   bIsVariable=true
   parent=CanvasPanel-LegacyParts
   widget=PartIconMain_Legacy
   properties {
    gradient=/FMJamSystem/Materials/MI_UI_Icon_Instrument_Type_faded
    imageSize=<42 42>
    renderTransform {angle=0, scale=<2.65 2.65>, shear=<0 0>, translation=<0 0>}
   }
   slot {
    class=CanvasPanelSlot
    properties {bAutoSize=true, layoutData {alignment=<0.5 0.5>, anchors {maximum=<0.5 0.5>, minimum=<0.5 0.5>}, offsets=<ltrb 0 0 0 0>}}
   }
  }
  CanvasPanel-AllParts {
   widgetClassPath=&CanvasPanel
   parent=PartSetVisibilitySwitcher
   widget=CanvasPanel-AllParts
   slot {class=CommonVisibilitySwitcherSlot}
  }
  PartIconNext2 {
   widgetClassPath=&BP_JamPartIcon
   bIsVariable=true
   parent=CanvasPanel-AllParts
   widget=PartIconNext2
   properties {colorAndOpacity=<rgba 1 1 1 0.25>, imageSize=<36 36>}
   slot {
    class=CanvasPanelSlot
    properties {bAutoSize=true, layoutData {alignment=<0.5 0.5>, anchors {maximum=<0.5 0.5>, minimum=<0.5 0.5>}, offsets=<ltrb -120 0 0 0>}}
   }
  }
  PartIconNext1 {
   widgetClassPath=&BP_JamPartIcon
   bIsVariable=true
   parent=CanvasPanel-AllParts
   widget=PartIconNext1
   properties {colorAndOpacity=<rgba 1 1 1 0.6>, imageSize=<48 48>}
   slot {
    class=CanvasPanelSlot
    properties {bAutoSize=true, layoutData {alignment=<0.5 0.5>, anchors {maximum=<0.5 0.5>, minimum=<0.5 0.5>}, offsets=<ltrb -78 0 0 0>}}
   }
  }
  PartIconPrevious2 {
   widgetClassPath=&BP_JamPartIcon
   bIsVariable=true
   parent=CanvasPanel-AllParts
   widget=PartIconPrevious2
   properties {colorAndOpacity=<rgba 1 1 1 0.25>, imageSize=<36 36>}
   slot {
    class=CanvasPanelSlot
    properties {bAutoSize=true, layoutData {alignment=<0.5 0.5>, anchors {maximum=<0.5 0.5>, minimum=<0.5 0.5>}, offsets=<ltrb 120 0 0 0>}}
   }
  }
  PartIconPrevious1 {
   widgetClassPath=&BP_JamPartIcon
   bIsVariable=true
   parent=CanvasPanel-AllParts
   widget=PartIconPrevious1
   properties {colorAndOpacity=<rgba 1 1 1 0.6>, imageSize=<48 48>}
   slot {
    class=CanvasPanelSlot
    properties {bAutoSize=true, layoutData {alignment=<0.5 0.5>, anchors {maximum=<0.5 0.5>, minimum=<0.5 0.5>}, offsets=<ltrb 78 0 0 0>}}
   }
  }
  PartIconMain {
   widgetClassPath=&BP_JamPartIcon
   bIsVariable=true
   parent=CanvasPanel-AllParts
   widget=PartIconMain
   properties {
    gradient=&MI_UI_Icon_Instrument_Type_faded
    imageSize=<42 42>
    renderTransform {angle=0, scale=<2.65 2.65>, shear=<0 0>, translation=<0 0>}
   }
   slot {
    class=CanvasPanelSlot
    properties {bAutoSize=true, layoutData {alignment=<0.5 0.5>, anchors {maximum=<0.5 0.5>, minimum=<0.5 0.5>}, offsets=<ltrb 0 0 0 0>}}
   }
  }
  InputReflectorButton {
   widgetClassPath=/FortUILibrary/Input/InputReflectorButton
   bIsVariable=true
   parent=Main_Overlay
   widget=InputReflectorButton
   properties {
    buttonText=""
    contentAlignment=HAlign_Center
    holdData=&CommonUI_HoldData_Base_C
    onButtonBaseClicked=(&self:WidgetTree.InputReflectorButton.HandleOnButtonClicked)
    onButtonBaseDoubleClicked=(&self:WidgetTree.InputReflectorButton.HandleOnButtonDoubleClicked)
    onButtonBaseHovered=(&self:WidgetTree.InputReflectorButton.HandleOnButtonHovered)
    onButtonBaseUnhovered=(&self:WidgetTree.InputReflectorButton.HandleOnButtonUnhovered)
    onSelectedChangedBase=(&self:WidgetTree.InputReflectorButton.HandleOnSelectedChanged)
    style=/FortUILibrary/Foundation/Buttons/ButtonStyles/ButtonStyle-Clear.ButtonStyle-Clear_C
    hoveredSlateSoundOverride {resourceObject=null}
    pressedSlateSoundOverride {resourceObject=null}
    triggeringInputAction {dataTable=/FMJamSystem/Input/SharedAssets/JamBindings, rowName=SwapJamPart}
   }
   slot {class=OverlaySlot, properties {horizontalAlignment=HAlign_Center, verticalAlignment=VAlign_Center, padding=<ltrb 7.5 53 0 0>}}
  }
  SizeBox_InstrumentText {
   widgetClassPath=&SizeBox
   bIsVariable=true
   parent=Main_Overlay
   widget=SizeBox_InstrumentText
   properties {bOverride_WidthOverride=true, widthOverride=260}
   slot {class=OverlaySlot, properties {horizontalAlignment=HAlign_Center, verticalAlignment=VAlign_Center, padding=<ltrb 12 86 12 0>}}
  }
  Instrument_Text {
   widgetClassPath=&CommonTextBlock
   bIsVariable=true
   parent=SizeBox_InstrumentText
   widget=Instrument_Text
   properties {
    scrollStyle=&TextScrollStyle-Base_C
    text=Text Block
    font {
     bForceMonospaced=false
     bMaterialIsStencil=false
     fontMaterial=null
     fontObject=&HeadingNow
     letterSpacing=0
     monospacedWidth=1
     size=22.5
     skewAmount=0
     typefaceFontName=86 Bold Italic
     outlineSettings {
      bApplyOutlineToDropShadows=false
      bMiteredCorners=true
      bSeparateFillAlpha=false
      outlineMaterial=null
      outlineSize=2
      outlineColor=<rgba 0 0 0 0.4>
     }
    }
    shadowOffset=<0 0>
    strikeBrush {
     bIsDynamicallyLoaded=false
     drawAs=NoDrawType
     imageType=NoImage
     mirroring=NoMirror
     resourceName=None
     resourceObject=null
     tiling=NoTile
     imageSize=<32 32>
     margin=<ltrb 0 0 0 0>
     outlineSettings {
      bUseBrushTransparency=false
      roundingType=HalfHeightRadius
      width=0
      color {colorUseRule=UseColor_Specified, specifiedColor=<rgba 0 0 0 0>}
      cornerRadii=<0 0 0 0>
     }
     tintColor {colorUseRule=UseColor_Specified, specifiedColor=<rgba 1 1 1 1>}
     uVRegion {bIsValid=false, max=<0 0>, min=<0 0>}
    }
   }
   slot {class=SizeBoxSlot, properties {horizontalAlignment=HAlign_Center}}
  }
  SizeBox_Switcher {
   widgetClassPath=&SizeBox
   bIsVariable=true
   parent=Main_Overlay
   widget=SizeBox_Switcher
   properties {bOverride_HeightOverride=true, bOverride_WidthOverride=true, heightOverride=105, widthOverride=105}
   slot {class=OverlaySlot, properties {horizontalAlignment=HAlign_Center, verticalAlignment=VAlign_Center}}
  }
  Switcher_Button {
   widgetClassPath=&Button
   parent=SizeBox_Switcher
   widget=Switcher_Button
   properties {
    isFocusable=false
    widgetStyle {
     clickedSlateSound {resourceObject=null}
     disabled {
      bIsDynamicallyLoaded=false
      drawAs=NoDrawType
      imageType=NoImage
      mirroring=NoMirror
      resourceName=None
      resourceObject=null
      tiling=NoTile
      imageSize=<0 0>
      margin=<ltrb 0 0 0 0>
      outlineSettings {
       bUseBrushTransparency=false
       roundingType=HalfHeightRadius
       width=0
       color {colorUseRule=UseColor_Specified, specifiedColor=<rgba 0 0 0 0>}
       cornerRadii=<0 0 0 0>
      }
      tintColor {colorUseRule=UseColor_Specified, specifiedColor=<rgba 1 1 1 1>}
      uVRegion {bIsValid=false, max=<0 0>, min=<0 0>}
     }
     disabledForeground {colorUseRule=UseColor_Specified, specifiedColor=<rgba 0.5271 0.5271 0.5271 1>}
     hovered {
      bIsDynamicallyLoaded=false
      drawAs=NoDrawType
      imageType=NoImage
      mirroring=NoMirror
      resourceName=None
      resourceObject=/FortUILibrary/Foundation/Art_ToolKit/Tools/M_UI_Circle_White
      tiling=NoTile
      imageSize=<513 512>
      margin=<ltrb 0 0 0 0>
      outlineSettings {
       bUseBrushTransparency=true
       roundingType=FixedRadius
       width=6
       color {colorUseRule=UseColor_Specified, specifiedColor=<rgba 0.7243 0.7243 0.7243 0>}
       cornerRadii=<176 176 176 176>
      }
      tintColor {colorUseRule=UseColor_Specified, specifiedColor=<rgba 0 0 0 0.15>}
      uVRegion {bIsValid=false, max=<0 0>, min=<0 0>}
     }
     hoveredForeground {colorUseRule=UseColor_Specified, specifiedColor=<rgba 0 1 1 1>}
     hoveredSlateSound {resourceObject=null}
     normal {
      bIsDynamicallyLoaded=false
      drawAs=NoDrawType
      imageType=NoImage
      mirroring=NoMirror
      resourceName=None
      resourceObject=&M_UI_Circle_White
      tiling=NoTile
      imageSize=<352 352>
      margin=<ltrb 0 0 0 0>
      outlineSettings {
       bUseBrushTransparency=true
       roundingType=FixedRadius
       width=1
       color {colorUseRule=UseColor_Specified, specifiedColor=<rgba 0.6951 0.6951 0.6951 1>}
       cornerRadii=<4 4 4 4>
      }
      tintColor {colorUseRule=UseColor_Specified, specifiedColor=<rgba 0 0 0 0.15>}
      uVRegion {bIsValid=false, max=<0 0>, min=<0 0>}
     }
     normalForeground {colorUseRule=UseColor_Specified, specifiedColor=<rgba 0 1 1 1>}
     normalPadding=<ltrb 12 1.5 12 1.5>
     pressed {
      bIsDynamicallyLoaded=false
      drawAs=NoDrawType
      imageType=NoImage
      mirroring=NoMirror
      resourceName=None
      resourceObject=&M_UI_Circle_White
      tiling=NoTile
      imageSize=<352 352>
      margin=<ltrb 0 0 0 0>
      outlineSettings {
       bUseBrushTransparency=true
       roundingType=FixedRadius
       width=6
       color {colorUseRule=UseColor_Specified, specifiedColor=<rgba 0.7243 0.7243 0.7243 1>}
       cornerRadii=<176 176 176 176>
      }
      tintColor {colorUseRule=UseColor_Specified, specifiedColor=<rgba 0 0 0 0.15>}
      uVRegion {bIsValid=false, max=<0 0>, min=<0 0>}
     }
     pressedForeground {colorUseRule=UseColor_Specified, specifiedColor=<rgba 0 1 1 1>}
     pressedPadding=<ltrb 12 1.5 12 1.5>
     pressedSlateSound {resourceObject=null}
    }
   }
   slot {class=SizeBoxSlot, properties {horizontalAlignment=HAlign_Center, verticalAlignment=VAlign_Center}}
  }
 ]
}
