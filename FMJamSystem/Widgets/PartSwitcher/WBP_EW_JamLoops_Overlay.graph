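# Jam Loops emote wheel overlay that shows the linear instrument part switcher alongside a bottom-anchored song info panel.
# Activation and deactivation play a short opacity fade forward or in reverse.
# It caches the owning player's jam session event router on activation and broadcasts part switcher displayed true, then false on deactivation.
# SetLayout uses its ShowContainer flag to move the down-only scaled switcher between padded center and top-left placement.
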WidgetBlueprint /FMJamSystem/Widgets/PartSwitcher/WBP_EW_JamLoops_Overlay
animations [Intro {
  binding_count=1
  length_seconds=0.15
  object=&self:Intro
  sequence {bindings [Overlay_0 {
     id=#1
     possessed_class=Overlay
     tracks [{
       type=MovieSceneFloatTrack
       display_name=Render Opacity
       sections [{type=MovieSceneFloatSection, end_frame=3, start_frame=0, channels [None {type=MovieSceneScriptingFloatChannel, keys [[0,0],[3,1]]}]}]
      }]
    }]}
 }]
graphs [
 EventGraph {nodes [
   Event On Deactivated()
   n1 = Get Intro()
   > Check and Stop Animation!(execute:Exec, Animation:Widget Animation Object Reference=@n1) -> (then:Exec) {is=title K2Node_MacroInstance}
   > PlayAnimationReverse(InAnimation=@n1)
   > n2 = Get!(execute:Exec) -> (then:Exec, else:Exec, EventRouter:Gameplay Event Router Component Object Reference)
   > Broadcast Event!(
    execute
    self:Gameplay Event Router Component Object Reference=@n2
    EventData:Jam Event Part Switcher Displayed Structure=Make Jam Event Part Switcher Displayed!(isDisplayed_1_#2:bool=False) -> (JamEvent_PartSwitcher_Displayed:Jam Event Part Switcher Displayed Structure) {is=title K2Node_MakeStruct}
    EventContext:Object Reference
   ) -> (then, ReturnValue:bool)
   Event On Activated()
   n3 = Get Intro()
   > Check and Stop Animation(Animation=@n3)
   > PlayAnimationForward(InAnimation=@n3)
   > n4 = GetJamSessionForActor!(execute, Actor:Actor Object Reference=GetOwningPlayer()) -> (then, ReturnValue:Jam Session Interface Interface)
   > Branch(Condition=IsValidInterface!(Interface:"Interface Interface (by ref)"=@n4) -> (ReturnValue:bool))
   n5 = GetGameplayEventRouterComponent!(self:Jam Session Interface Interface=@n4) -> (ReturnValue:Gameplay Event Router Component Object Reference)
   > Is Valid(InputObject=@n5)
   > n6 = Set EventRouter(EventRouter:Gameplay Event Router Component Object Reference=@n5) -> (Output_Get:Gameplay Event Router Component Object Reference)
   > Broadcast Event(self=@n6, EventData=Make Jam Event Part Switcher Displayed(isDisplayed_1_#2=true))
  ]}
 SetLayout {nodes [
   n1 = SetLayout!() -> (then, ShowContainer:bool=~) {is=title K2Node_FunctionEntry}
   n2 = SlotAsOverlaySlot(Widget=Get InstumentSwitcher_ScaleBox() -> (InstumentSwitcher_ScaleBox:Scale Box Object Reference))
   > SetHorizontalAlignment(
    self=@n2
    InHorizontalAlignment=Select(Option 0:EHorizontalAlignment Enum=HAlign_Center, Option 1:EHorizontalAlignment Enum=HAlign_Left, Index=@n1) -> (ReturnValue:EHorizontalAlignment Enum)
   )
   > SetVerticalAlignment(
    self=@n2
    InVerticalAlignment=Select(Option 0:EVerticalAlignment Enum=VAlign_Center, Option 1:EVerticalAlignment Enum=VAlign_Top, Index=@n1) -> (ReturnValue:EVerticalAlignment Enum)
   )
   > SetPadding!(
    execute
    self:Overlay Slot Object Reference=@n2
    InPadding:Margin Structure=Select(
     Option 0:Margin Structure=Make Margin(Left=0, Top=200, Right=0, Bottom=200)
     Option 1:Margin Structure=Make Margin(Left=58, Top=88, Right=0, Bottom=0)
     Index=@n1
    ) -> (ReturnValue:Margin Structure)
   ) -> (then)
  ]}
]
variables [EventRouter {type=GameplayEventRouterComponent, default=null}]
widgets {
 info {inheritedWidgetCount=0, parentClass=/Script/FMJamUI.JamEmoteWheelOverlay, rootWidgetClass=&Overlay, widgetCount=5}
 widgets [
  Overlay_0 {widgetClassPath=&Overlay, widget=Overlay_0, properties {bOverride_Cursor=true}}
  SizeBox_JamSongInfo {
   widgetClassPath=&SizeBox
   bIsVariable=true
   parent=Overlay_0
   widget=SizeBox_JamSongInfo
   properties {bOverride_WidthOverride=true, widthOverride=400}
   slot {class=OverlaySlot, properties {verticalAlignment=VAlign_Bottom, padding=<ltrb 30 0 0 80>}}
  }
  WBP_JamSongInfo {
   widgetClassPath=/FMJamSystem/Widgets/SongInfo/WBP_JamSongInfo
   bIsVariable=true
   parent=SizeBox_JamSongInfo
   widget=WBP_JamSongInfo
   properties {condensedView=false}
   slot {class=SizeBoxSlot}
  }
  InstumentSwitcher_ScaleBox {
   widgetClassPath=&ScaleBox
   bIsVariable=true
   parent=Overlay_0
   widget=InstumentSwitcher_ScaleBox
   properties {stretchDirection=DownOnly}
   slot {class=OverlaySlot, properties {horizontalAlignment=HAlign_Center, verticalAlignment=VAlign_Center, padding=<ltrb 0 200 0 200>}}
  }
  BP_JamPartSwitcher_Linear {
   widgetClassPath=&/WBP_JamPartSwitcher_Linear
   bIsVariable=true
   parent=InstumentSwitcher_ScaleBox
   widget=BP_JamPartSwitcher_Linear
   slot {class=ScaleBoxSlot, properties {horizontalAlignment=HAlign_Fill, verticalAlignment=VAlign_Fill}}
  }
 ]
}
